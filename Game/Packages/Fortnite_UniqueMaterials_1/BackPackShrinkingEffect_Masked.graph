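# Masked, two-sided Default Lit material for skeletal meshes, named for a backpack shrinking effect.
# FresnelFunction processes diffuse and emissive textures, with an optional SuperChargedGlow switch and a Hologram term added for emissive.
# LODEffect, driven by a pickup distance-cull override, offsets world position into BlackHoleFunction with a black hole location parameter, apparently for vertex deformation.
# Texture parameters supply normals, the mask and packed specular, metallic and remappable roughness.
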Material /Game/Packages/Fortnite_UniqueMaterials_1/BackPackShrinkingEffect_Masked
graph {
 properties {
  bUsedWithMeshDeformer=true
  bUsedWithSkeletalMesh=true
  blendMode=BLEND_Masked
  pixelDepthOffsetMode=PDOM_Legacy
  twoSided=true
  usedShadingModels=Default Lit
 }
 nodes [
  n1 = VectorParameter() {group=Super Charge, parameterName=AverageHandPosition, defaultValue=<rgba 0 0 0 1>}
  n2 = ScalarParameter() {defaultValue=1, group=Super Charge, parameterName=SuperChargedFlickerBias}
  n3 = ScalarParameter() {group=Super Charge, parameterName=SuperChargedEffect}
  n4 = StaticBool()
  n5 = MaterialFunctionCall!(
   DiffuseFresnelAmount=ScalarParameter() {defaultValue=0.75, group=Diffuse, parameterName=DiffuseFresnelAmount}
   Emissive=Multiply(
    A=TextureSampleParameter2D() {group=Emissive, parameterName=EmissiveTexture, texture=&T_WhitesRGB}.RGB
    B=VectorParameter() {group=Emissive, parameterName=EmissiveMultiplier, defaultValue=<rgba 0 0 0 1>}.RGB
   )
   Base Color=TextureSampleParameter2D() {parameterName=Diffuse, texture=&T_BlacksRGB}.RGB
  ) {materialFunction=/Game/Packages/Fortress_SharedMaterials/Base_Material_Functions/FresnelFunction}
  Add(
   A=StaticSwitchParameter(
    True=MaterialFunctionCall!(AverageHandPosition=@n1.RGB, SuperChargedFlickerBias=@n2, SuperChargedEffect=@n3, UseDistanceCalc=@n4, Emissive=@n5.Emissive) {materialFunction=/Game/Characters/Fortress_CharacterShader/SuperChargedGlow}.Result
    False=Constant()
   ) {group=Super Charge, parameterName=Use Super Chared Effect}
   B=MaterialFunctionCall!(In) {materialFunction=/Game/Packages/Fortress_SharedMaterials/Base_Material_Functions/Hologram}.Emissive
  )
  n6 = VectorParameter() {parameterName=WorldPostionOfBlackHole, defaultValue=<rgba -30000 25536 1952 0>}
  n7 = MaterialFunctionCall!(
   World Position
   DistanceCullFadeAnimationOffset=MaterialFunctionCall() {materialFunction=/Game/Packages/Fortress_SharedMaterials/Base_Material_Functions/PickupEffectDistanceCullOverride}.Result
   Use actor center instead of pivot point
  ) {materialFunction=/Game/Packages/Fortress_SharedMaterials/Base_Material_Functions/LODEffect}
  n8 = WorldPosition()
  Add(B=MaterialFunctionCall!(WorldPostionOfBlackHole=AppendVector(A=@n6.RGB, B=@n6.RGB), World Position=Add(A=@n7.Result, B=@n8.XYZ)) {materialFunction=&/BlackHoleFunction}.Result)
  n9 = TextureSampleParameter2D() {
   desc="r= spec g= metallic B = roughness\n"
   group=Specular
   parameterName=SpecularMasks
   samplerType=SAMPLERTYPE_Masks
   texture=&T_Fortnite_Default_S
  }
  LinearInterpolate(
   A=ScalarParameter() {group=Specular, parameterName=SpecRoughnessMin}
   B=ScalarParameter() {defaultValue=1, group=Specular, parameterName=SpecRoughnessMax}
   Alpha=@n9.B
  )
  n10 = TextureSampleParameter2D() {group=Normal, parameterName=Normals, samplerType=SAMPLERTYPE_Normal, texture=&FlatNormalBlue}
  Reroute(None=@n10.RGB)
  TextureSampleParameter2D() {parameterName=Mask, texture=&T_WhitesRGB}
 ]
}
