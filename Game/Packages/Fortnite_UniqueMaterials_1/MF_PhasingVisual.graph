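# Material function that adds a phasing visual to a material's Emissive and BaseColor, toggled per primitive through the IsPhasing custom primitive data parameter.
# A Fresnel term blends two warm colors and adds to a square grayscale texture projected on world XY, scaled by PhaseSize.
# The resulting glow is added to Emissive, and its clamped red channel darkens BaseColor.
# A time-based sine pulse of the pattern is built but left unconnected.
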MaterialFunction /Game/Packages/Fortnite_UniqueMaterials_1/MF_PhasingVisual
graph {nodes [
  n1 = Multiply(A=Fresnel() {baseReflectFraction=0.01, exponent=1}) {constB=0.5}
  n2 = LinearInterpolate(
   A=VectorParameter() {defaultValue=<rgba 0.45 0.1858 0.0155 1>}.RGB
   B=VectorParameter() {parameterName=None2, defaultValue=<rgba 0.6 0.5034 0 1>}.RGB
   Alpha=@n1
  )
  n3 = WorldPosition()
  n4 = ScalarParameter() {defaultValue=128, parameterName=PhaseSize}
  n5 = TextureSample(UVs=ComponentMask(None=Divide(A=@n3.XYZ, B=@n4)) {g=true, r=true}) {samplerType=SAMPLERTYPE_Grayscale, texture=&/T_PhaseSquare}
  n6 = Add(A=@n1, B=@n5.R)
  n7 = Multiply(
   A=Multiply(A=Multiply(A=@n2, B=@n6)) {constB=0.5}
   B=ScalarParameter() {bUseCustomPrimitiveData=true, parameterName=IsPhasing, primitiveDataIndex=14}
  )
  FunctionOutput(None=Add(A=FunctionInput() {bUsePreviewValueAsDefault=true, inputName=Emissive}, B=@n7)) {outputName=Emissive}
  n8 = MaterialFunctionCall(Float3=Clamp(None=@n7)) {materialFunction=&BreakOutFloat3Components}
  FunctionOutput(None=Multiply(A=FunctionInput() {bUsePreviewValueAsDefault=true, inputName=BaseColor}, B=LinearInterpolate(Alpha=@n8.R) {constA=1, constB=0.5})) {outputName=BaseColor}
  n9 = Divide(A=@n5.R) {constB=4}
  n10 = Time() {bOverride_Period=true, period=2}
  n11 = Sine(None=Multiply(A=Add(A=@n9, B=@n10)) {constB=-1})
  Multiply(A=@n5.R, B=Clamp(None=ConstantBiasScale(None=@n11)))
 ]}
