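# Material function that unpacks a normal and per-vertex build animation progress from mesh UV channels, apparently for building-piece construction animation.
# It appends the red component of TexCoord 4 to the two components of TexCoord 3 for the Normals output.
# The green component of TexCoord 4 feeds the Per Vert Build Animation Progress output.
# It samples no textures and takes no inputs, relying entirely on data stored in the mesh.
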MaterialFunction /Game/Packages/Fortnite_UniqueMaterials_1/MaterialFunctions/UpdatedNormalAndPixelBuildLevel
graph {nodes [
  n1 = TextureCoordinate() {coordinateIndex=4}
  FunctionOutput(None=AppendVector(A=TextureCoordinate() {coordinateIndex=3}, B=ComponentMask(None=@n1) {r=true}) {desc=animation phase}) {outputName=Normals}
  FunctionOutput(None=ComponentMask(None=@n1) {g=true}) {outputName=Per Vert Build Animation Progress, sortPriority=1}
 ]}
