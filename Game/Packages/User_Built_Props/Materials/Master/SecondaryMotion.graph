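# Material function for user-built props that offsets an animated gradient with a secondary motion falloff, apparently for build and damage animation.
# The falloff comes from SelfBuildingWalls2, driven by the SecondaryMotionFalloff input.
# One branch subtracts falloff, faded in near gradient zero; the other adds it, faded out near both ends.
# Code_BeingDamaged blends from subtracting to adding, and the result is saturated.
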MaterialFunction /Game/Packages/User_Built_Props/Materials/Master/SecondaryMotion
graph {nodes [
  n1 = Saturate(None=FunctionInput() {bUsePreviewValueAsDefault=true, inputName=AnimatedGradientValues, inputType=FunctionInput_Scalar, sortPriority=10})
  n2 = FunctionInput() {
   bUsePreviewValueAsDefault=true
   inputName=SecondaryMotionFalloff
   inputType=FunctionInput_Scalar
   sortPriority=10
   previewValue=<1596.8654 0 0 1>
  }
  n3 = MaterialFunctionCall!(Number of sub objects, Manual Anim Grad, Sec mot Falloff rad=@n2) {materialFunction=/Game/Effects/Fort_Effects/Materials/Temp/SelfBuildingWalls2}
  n4 = Multiply(A=@n1) {constB=10}
  n5 = Subtract(A=@n1, B=Multiply(A=@n3.falloff, B=Saturate(None=@n4)))
  n6 = Saturate(None=Multiply(A=OneMinus(None=@n1)) {constB=5})
  n7 = Saturate(None=Multiply(A=@n1) {constB=5})
  n8 = Add(A=@n1, B=Multiply(A=Multiply(A=@n6, B=@n7), B=@n3.falloff))
  n9 = FunctionInput() {bUsePreviewValueAsDefault=true, inputName=Code_BeingDamaged, inputType=FunctionInput_Scalar, sortPriority=10, previewValue=<1596.8654 0 0 1>}
  FunctionOutput(None=Saturate(None=LinearInterpolate(A=@n5, B=@n8, Alpha=@n9)))
  OneMinus(None=@n6)
 ]}
