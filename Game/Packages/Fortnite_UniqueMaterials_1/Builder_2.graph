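# Master material for buildable walls and floors that animates their construction, damage and destruction board by board.
# Custom primitive data and pivot-based functions drive world position offset that moves and rotates pieces, with optional distance-based build waves.
# Lerped texture sets, world-position color picking and height-lerped build masks feed shading, with banner sweep and blueprint emissive layered on.
# A MobileVersion switch drops the offset, using a masked-texture reveal instead.
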Material /Game/Packages/Fortnite_UniqueMaterials_1/Builder_2
graph {
 properties {
  bAutomaticallySetUsageInEditor=false
  bUseMaterialAttributes=true
  bUsedWithNanite=true
  numCustomizedUVs=5
  pixelDepthOffsetMode=PDOM_Legacy
  usedShadingModels=Default Lit
 }
 nodes [
  n1 = Constant()
  n2 = ComponentMask(None=ActorPositionWS()) {b=true, desc=Needs actor world position}
  n3 = ComponentMask(None=WorldPosition().XYZ) {b=true}
  n4 = Subtract(A=Subtract(A=@n2, B=ScalarParameter() {defaultValue=-90, group=Build_Mask_Settings, parameterName=Gradient offset Position Lower}), B=@n3)
  n5 = ScalarParameter() {defaultValue=90, group=Build_Mask_Settings, parameterName=OpacityGradientSize}
  n6 = OneMinus(None=Clamp(None=Divide(A=@n4, B=@n5)))
  n7 = ScalarParameter() {defaultValue=-256, group=Build_Mask_Settings, parameterName=Gradient Offset Upper Position}
  n8 = Divide(
   A=Subtract(A=Subtract(A=@n2, B=@n7), B=@n3)
   B=ScalarParameter() {defaultValue=90, group=Build_Mask_Settings, parameterName=OpacityGradientSize}
  )
  n9 = StaticSwitchParameter(True=@n6, False=Multiply(A=Clamp(None=@n8), B=@n6)) {desc=Adds additional mask, group=Build, parameterName=Walls True Floors False}
  n10 = Transform(None=ReflectionVectorWS())
  n11 = OneMinus(None=ComponentMask(None=ConstantBiasScale(None=@n10)) {g=true, r=true})
  n12 = TextureSampleParameter2D() {group=Diffuse, parameterName=Diffuse, texture=/Game/Packages/PBW/Wood/Textures/PBW_wood_2_D}
  n13 = Fresnel()
  n14 = TextureSampleParameter2D() {
   group=Specular
   parameterName=SpecularMasks
   samplerType=SAMPLERTYPE_Masks
   texture=&T_Fortnite_Default_S
  }
  n15 = AppendVector(A=@n12.R, B=@n12.R)
  n16 = TextureSampleParameter2D() {group=Normals, parameterName=NormalMap, samplerType=SAMPLERTYPE_Normal, texture=&FlatNormalBlue}
  n17 = MaterialFunctionCall!(Use UV 2=StaticBoolParameter() {group=Multiple_Material_Textures, parameterName=Mat 2 Use UV2}) {materialFunction=/Game/Packages/Fortress_SharedMaterials/Base_Material_Functions/UseUV1or2}
  n18 = TextureSampleParameter2D(UVs=@n17.Result) {
   desc=Use for interior walls and ceilings
   group=Material_2
   parameterName=Normals_Texture_2
   samplerSource=SSM_Wrap_WorldGroupSettings
   samplerType=SAMPLERTYPE_Normal
   texture=/Engine/EngineMaterials/FlatNormal
  }
  n19 = TextureSampleParameter2D(UVs=@n17.Result) {group=Diffuse, parameterName=Diffuse2, samplerSource=SSM_Wrap_WorldGroupSettings, texture=&PBW_wood_2_D}
  n20 = TextureSampleParameter2D(UVs=@n17.Result) {
   desc=Use for interior walls and ceilings
   group=Material_2
   parameterName=SpecularMasks_2
   samplerSource=SSM_Wrap_WorldGroupSettings
   samplerType=SAMPLERTYPE_Masks
   texture=&T_Fortnite_Default_S
  }
  n21 = MaterialFunctionCall(Use UV 2=StaticBoolParameter() {group=Multiple_Material_Textures, parameterName=Mat 3 Use UV2}) {materialFunction=&UseUV1or2}
  n22 = TextureSampleParameter2D(UVs=@n21.Result) {group=Diffuse, parameterName=Diffuse3, samplerSource=SSM_Wrap_WorldGroupSettings, texture=&PBW_wood_2_D}
  n23 = TextureSampleParameter2D(UVs=@n21.Result) {
   desc=Use for trims and accents
   group=Material_3
   parameterName=SpecularMasks_3
   samplerSource=SSM_Wrap_WorldGroupSettings
   samplerType=SAMPLERTYPE_Masks
   texture=&T_Fortnite_Default_S
  }
  n24 = TextureSampleParameter2D(UVs=@n21.Result) {
   desc=Use for trims and accents
   group=Material_3
   parameterName=Normals_Texture_3
   samplerSource=SSM_Wrap_WorldGroupSettings
   samplerType=SAMPLERTYPE_Normal
   texture=&FlatNormal
  }
  n25 = MaterialFunctionCall(Use UV 2=StaticBoolParameter() {group=Multiple_Material_Textures, parameterName=Mat 4 Use UV2}) {materialFunction=&UseUV1or2}
  n26 = TextureSampleParameter2D(UVs=@n25.Result) {group=Diffuse, parameterName=Diffuse4, samplerSource=SSM_Wrap_WorldGroupSettings, texture=&PBW_wood_2_D}
  n27 = TextureSampleParameter2D(UVs=@n25.Result) {
   desc=Use for doors
   group=Material_4
   parameterName=SpecularMasks_4
   samplerSource=SSM_Wrap_WorldGroupSettings
   samplerType=SAMPLERTYPE_Masks
   texture=&T_Fortnite_Default_S
  }
  n28 = TextureSampleParameter2D(UVs=@n25.Result) {
   desc=Use for doors
   group=Material_4
   parameterName=Normals_Texture_4
   samplerSource=SSM_Wrap_WorldGroupSettings
   samplerType=SAMPLERTYPE_Normal
   texture=&FlatNormal
  }
  n29 = TextureSampleParameter2D() {group=Diffuse, parameterName=Diffuse5, samplerSource=SSM_Wrap_WorldGroupSettings, texture=&PBW_wood_2_D}
  n30 = TextureSampleParameter2D() {group=Diffuse, parameterName=Diffuse6, samplerSource=SSM_Wrap_WorldGroupSettings, texture=&PBW_wood_2_D}
  n31 = TextureCoordinate() {coordinateIndex=1}
  n32 = TextureCoordinate() {uTiling=0.5}
  n33 = MaterialFunctionCall!(
   Spec1=@n14.RGB
   Texture1=@n15
   Normal1=StaticSwitchParameter(True=@n16.RGB, False=Constant3Vector() {constant=<rgba 0 0 1 0>}) {group=Normals, parameterName=UseNormalMap}
   Normal2=@n18.RGB
   Texture2=AppendVector(A=@n19.RGB, B=@n19.RGB)
   Spec2=@n20.RGB
   Texture3=AppendVector(A=@n22.RGB, B=@n22.RGB)
   Spec3=@n23.RGB
   Normal3=@n24.RGB
   Texture4=AppendVector(A=@n26.RGB, B=@n26.RGB)
   Spec4=@n27.RGB
   Normal4=@n28.RGB
   Texture5=AppendVector(A=@n29.RGB, B=@n29.RGB)
   Normal5
   Spec5
   Texture6=AppendVector(A=@n30.RGB, B=@n30.RGB)
   Spec6
   Normal6
   Texture7
   Texture8
   Texture9
   Use_Dot_Texture9
   Texture10
   Use 2 Diffuse Textures=StaticBoolParameter() {group=Diffuse, parameterName=Use 2 Materials}
   --------------
   Use 3 Diffuse Textures=StaticBoolParameter() {group=Diffuse, parameterName=Use 3 Materials}
   Use 4 Diffuse Textures=StaticBoolParameter() {group=Diffuse, parameterName=Use 4 Materials}
   Use 5 Diffuse Textures=StaticBoolParameter() {group=Diffuse, parameterName=Use 5 Materials}
   Use 6 Diffuse Textures=StaticBoolParameter() {group=Diffuse, parameterName=Use 6 Materials}
   Use 7 Diffuse Textures
   Use 8 Diffuse Textures
   Use 9 Diffuse Textures
   Use 10 Diffuse Textures
   "-------------- "
   Dot_Texture1
   TransitionInfo=StaticSwitchParameter(True=ComponentMask(None=@n31) {r=true}, False=ComponentMask(None=@n32) {r=true}) {parameterName=MobileVersion}
   Dot_Texture2
   Dot_Texture3
   Dot_Texture4
   Dot_Texture5
   Dot_Texture6
   Dot_Texture7
   Dot_Texture8
   Dot_Texture9
   Dot_Texture10
   Use_Dot_Texture1
   Use_Dot_Texture2
   Use_Dot_Texture3
   Use_Dot_Texture4
   Use_Dot_Texture5
   Use_Dot_Texture6
   Use_Dot_Texture7
   Use_Dot_Texture8
   Use_Dot_Texture10
  ) {materialFunction=/Game/Packages/Fortress_SharedMaterials/Base_Material_Functions/Fortnite_TextureSetLerp}
  n34 = StaticSwitchParameter(True=@n33.Diffuse, False=@n15) {group=Diffuse, parameterName=Use multiple diffuse textures}
  n35 = ComponentMask(None=@n34) {b=true}
  n36 = OneMinus(None=Multiply(A=OneMinus(None=@n13), B=OneMinus(None=@n35)))
  n37 = Multiply(A=@n36, B=@n36)
  n38 = Ceil(None=ComponentMask(None=TextureCoordinate()) {g=true})
  n39 = ScalarParameter() {
   defaultValue=0.1
   desc="1/number of colors. ie 1/10 = .1"
   group=World Space Color Picker
   parameterName="Number Of Colors in Color Chart (1/Number of colors)"
  }
  n40 = MaterialFunctionCall!(
   f2
   f3=Divide(A=ActorPositionWS(), B=ScalarParameter() {defaultValue=16.71, group=World Space Color Picker, parameterName="Color Change Frequency (WS)"})
   f4
  ) {materialFunction=&/VectorComponentOps}
  n41 = Add(A=Multiply(A=Subtract(A=@n38) {constB=0.5}, B=@n39), B=@n40.f3 Add)
  n42 = ComponentMask(None=@n34) {b=true, g=true, r=true}
  n43 = Multiply(
   A=TextureSampleParameter2D(UVs=StaticSwitchParameter(True=@n41, False=@n40.f3 Add) {group=World Space Color Picker, parameterName=Offset Color Index with UV Channel 0 G}) {group=World Space Color Picker, parameterName=Color Chart, texture=/Game/Global/Textures/Default/ColorSets/CarColors_Interiors}.RGB
   B=@n42
  )
  n44 = TextureSampleParameter2D(UVs=@n40.f3 Add) {group=World Space Color Picker, parameterName=Color Chart 2, texture=/Game/Global/Textures/Default/ColorSets/CarColors}
  n45 = TextureSampleParameter2D() {
   group=World Space Color Picker
   parameterName=World Space Color Picker Mask
   texture=/Game/Vehicles/FORT_Vehicles_01/Materials/M_FORT_Truck_01_M_New
  }
  n46 = AppendVector(A=@n45.RGB, B=@n45.RGB)
  n47 = StaticComponentMaskParameter(None=@n46) {defaultB=true, group=World Space Color Picker, parameterName=World Space Color Picker Mask Channel}
  n48 = LinearInterpolate(
   A=LinearInterpolate(A=@n43, B=Multiply(A=@n44.RGB, B=@n42), Alpha=@n47)
   B=@n42
   Alpha=StaticComponentMaskParameter(None=@n46) {defaultG=true, group=World Space Color Picker, parameterName=World Space Color Picker 2 Mask Channel}
  )
  n49 = LinearInterpolate(A=@n43, B=@n42, Alpha=@n47)
  n50 = StaticSwitchParameter(
   True=StaticSwitchParameter(
    True=StaticSwitchParameter(True=@n48, False=@n49) {group=World Space Color Picker, parameterName=Use 2 World Position Coloring Colors}
    False=@n43
   ) {group=World Space Color Picker, parameterName=Mask World Position Color Picker}
   False=@n42
  ) {desc="the picker uses ", group=World Space Color Picker, parameterName=Use World Position Coloring}
  n51 = VectorParameter() {group=Diffuse, parameterName=DiffuseMultiplier, defaultValue=<rgba 1 1 1 1>}
  n52 = StaticSwitchParameter(
   True=LinearInterpolate(A=TextureSample(UVs=@n11) {texture=/Game/Packages/User_Built_Props/Textures/ReflectionMap4}.RGB, B=@n12.R, Alpha=OneMinus(None=@n37))
   False=StaticSwitchParameter(True=Multiply(A=@n50, B=@n51.RGB), False=@n50) {group=Diffuse, parameterName=ModulateDiffuse}
  ) {group=reflection, parameterName=Use Reflection}
  n53 = VectorParameter() {group=Build_Mask_Settings, parameterName=FormingTextureColor, defaultValue=<rgba 1 0.7905 0.5851 1>}
  n54 = Add(
   A=@n9
   B=MaterialFunctionCall() {materialFunction=/Game/Packages/Fortnite_UniqueMaterials_1/MaterialFunctions/UpdatedNormalAndPixelBuildLevel}.Per Vert Build Animation Progress
  )
  n55 = ComponentMask(None=MaterialFunctionCall!(Lerp A, Lerp B, Height Texture=Constant(), Falloff Contrast, Transition Phase=Clamp(None=@n54)) {materialFunction=/Game/Packages/Fortnight_MaterialFunctions/Temp/HeightLerp}.Lerp Alpha) {desc="<- opacity", r=true}
  n56 = Multiply(A=Add(A=OneMinus(None=@n55), B=Constant() {r=1}), B=@n52)
  n57 = Constant() {r=1}
  n58 = Subtract(A=@n55, B=Constant() {r=0.45})
  n59 = Constant() {r=10}
  n60 = StaticSwitchParameter(
   True=@n57
   False=StaticSwitchParameter(
    True=MaterialFunctionCall(In=@n58, Contrast=@n59) {materialFunction=&CheapContrast}.Result
    False=@n57
   ) {defaultValue=true, group=Build, parameterName=Builder Material}
  ) {group=Build, parameterName=b Damaged but Not Animating Shader}
  n61 = StaticSwitchParameter(
   True=@n52
   False=StaticSwitchParameter(True=LinearInterpolate(A=@n53.RGB, B=@n56, Alpha=@n60) {desc=d}, False=@n52) {defaultValue=true, group=Build, parameterName=Builder Material}
  ) {parameterName=MobileVersion}
  n62 = MaterialFunctionCall(
   X=NamedRerouteUsage()
   Y=NamedRerouteUsage()
   Z=LinearInterpolate(
    A=ScalarParameter() {defaultValue=0.1, group=Specular, parameterName=SpecRoughnessMin}
    B=ScalarParameter() {defaultValue=1, group=Specular, parameterName=SpecRoughnessMax}
    Alpha=NamedRerouteUsage()
   )
  ) {materialFunction=&MakeFloat3}
  n63 = VectorParameter() {group=Specular, parameterName=SpecularMasks_FormingProperties, defaultValue=<rgba 0.5 0 1 1>}
  n64 = StaticSwitchParameter(
   True=@n62.Result
   False=StaticSwitchParameter(True=LinearInterpolate(A=@n63.RGB, B=@n62.Result, Alpha=@n60), False=@n62.Result) {defaultValue=true, group=Build, parameterName=Builder Material}
  ) {group=Build, parameterName=b Damaged but Not Animating Shader}
  n65 = TextureSampleParameter2D() {group=Blueprint_Info, parameterName=Blueprint_Outline_Texture, texture=/Game/Packages/Fortress_SharedMaterials/Textures/bluePrintOutlines1}
  n66 = VectorParameter() {group=Blueprint_Info, parameterName=Blueprint_Outline_Texture_Channel, defaultValue=<rgba 1 0 0 0>}
  n67 = Clamp(None=DotProduct(A=AppendVector(A=@n65.RGB, B=@n65.RGB), B=AppendVector(A=@n66.RGB, B=@n66.RGB)))
  n68 = VectorParameter() {bUseCustomPrimitiveData=true, group=Emissive, parameterName=Emissive_Color_Add, primitiveDataIndex=7, defaultValue=<rgba 0 0 0 1>}
  n69 = Multiply(A=@n68.RGB, B=ScalarParameter() {defaultValue=0.35, parameterName=EmissiveFlashIntensity})
  n70 = Add(A=@n67, B=@n69)
  n71 = StaticSwitchParameter(True=@n70, False=StaticSwitchParameter(True=@n70, False=@n69) {group=Blueprint_Info, parameterName=DebugImageSelection}) {group=Emissive, parameterName=Emissive_Glow}
  n72 = MaterialFunctionCall!(Number of sub objects, Manual Anim Grad, Sec mot Falloff rad) {materialFunction=/Game/Effects/Fort_Effects/Materials/Temp/SelfBuildingWalls2}
  n73 = WorldPosition()
  n74 = MaterialFunctionCall!(
   AnimPhase=ScalarParameter() {bUseCustomPrimitiveData=true, defaultValue=1, group=Build, parameterName=BuildAnimPhase, primitiveDataIndex=1}
   AnimGrad=@n72.Anim Order
  ) {desc=white when invisible, materialFunction=/Game/Packages/User_Built_Props/wallAnim}
  n75 = Multiply(A=Add(A=Constant3Vector() {constant=<rgba 0 0 -100000000 0>}, B=Subtract(A=@n72.Board Pivot Pos, B=@n73.XYZ)), B=@n74.Result)
  n76 = Subtract(
   A=MaterialFunctionCall() {materialFunction=&SelfBuildingWalls2}.Board Pivot Pos
   B=WorldPosition().XYZ
  )
  n77 = MaterialFunctionCall() {materialFunction=&SelfBuildingWalls2}
  n78 = Frac(None=@n77.Board Pivot Pos)
  n79 = Add(
   A=@n77.Random Per Mesh
   B=MaterialFunctionCall(f3=Abs(None=@n78)) {materialFunction=&VectorComponentOps}.f3 Avg
  )
  n80 = Add(
   A=Multiply(A=Subtract(A=@n79), B=ScalarParameter() {group=3dBuild, parameterName=RandomOffsetAmount})
   B=Distance(A=@n77.Board Pivot Pos, B=VectorParameter() {group=3dBuild, parameterName=BuildWSStartLocation, defaultValue=<rgba -30976 25600 1920 1>}.RGB)
  )
  n81 = ScalarParameter() {defaultValue=-512, group=3dBuild, parameterName=BuildRampupDistance}
  n82 = Clamp(None=Add(A=Divide(A=@n80, B=@n81) {constB=256}, B=ScalarParameter() {group=3dBuild, parameterName=3DAnimationProgress}))
  n83 = OneMinus(None=@n82)
  n84 = ScalarParameter() {bUseCustomPrimitiveData=true, defaultValue=34, group=Build, parameterName=TotalNumberOfSubObjects}
  n85 = Frac(None=Multiply(A=Time(), B=ScalarParameter() {defaultValue=0.05, group=Build, parameterName=debug build speed}))
  n86 = MaterialFunctionCall() {materialFunction=&SelfBuildingWalls2}
  n87 = ActorPositionWS()
  n88 = Divide(
   A=Distance(
    A=StaticSwitchParameter(True=@n86.Board Pivot Pos, False=@n87) {group=3dBuild, parameterName=PerSubobject3dAnimation}
    B=VectorParameter() {group=3dBuild, parameterName=BuildWSStartLocation, defaultValue=<rgba -30976 25600 1920 1>}.RGB
   )
   B=ScalarParameter() {defaultValue=-512, group=3dBuild, parameterName=BuildRampupDistance}
  ) {constB=256}
  n89 = ScalarParameter() {group=3dBuild, parameterName=3DAnimationProgress}
  n90 = Multiply(A=@n84, B=Clamp(None=Add(A=@n88, B=@n89)))
  n91 = ScalarParameter() {bUseCustomPrimitiveData=true, defaultValue=1, group=Build, parameterName=BuildAnimPhase, primitiveDataIndex=1}
  n92 = ScalarParameter() {defaultValue=1, group=Build, parameterName=Debug_BreakAnimSpeed}
  n93 = ScalarParameter() {defaultValue=1, group=Build, parameterName=Debug_TotalDestructionAnimSpeed}
  n94 = MaterialFunctionCall!(
   TotalNumberOfSubObjects=@n84
   BuildAnimPhase=StaticSwitchParameter(True=Multiply(A=@n84, B=@n85), False=StaticSwitchParameter(True=@n90, False=@n91) {group=3dBuild, parameterName=3dAnimationBuild}) {group=Build, parameterName=debug build motion}
   TargetBreakNumber=ScalarParameter() {bUseCustomPrimitiveData=true, defaultValue=1000, group=Build, parameterName=TargetBreakNumber, primitiveDataIndex=6}
   BreakAnim=StaticSwitchParameter(
    True=MaterialFunctionCall(Speed=@n92) {materialFunction=&TimeWithSpeedVariable}.Frac Time
    False=ScalarParameter() {bUseCustomPrimitiveData=true, group=Build, parameterName=BreakAnim, primitiveDataIndex=3}
   ) {group=Build, parameterName=DebugBreakAnim}
   Code_totalDestructionAmount=StaticSwitchParameter(
    True=MaterialFunctionCall(Speed=@n93) {materialFunction=&TimeWithSpeedVariable}.Frac Time
    False=ScalarParameter() {bUseCustomPrimitiveData=true, group=Build, parameterName=Code_totalDestructionAmount, primitiveDataIndex=5}
   ) {group=Build, parameterName=DebugTotalDestruction}
  ) {materialFunction=/Game/Packages/User_Built_Props/Materials/Master/AnimationMF}
  n95 = Add(
   A=ScalarParameter() {group=3dBuild, parameterName=3dOverrideValue}
   B=StaticSwitchParameter(True=StaticSwitchParameter(True=@n83, False=@n82) {group=3dBuild, parameterName=BuildOrDamage}, False=@n94.single) {group=3dBuild, parameterName=PerSubobject3dAnimation}
  )
  n96 = MaterialFunctionCall!(
   SecondaryMotionFalloff=ScalarParameter() {defaultValue=-27755, group=Build, parameterName=SecondaryMotionFalloff}
   AnimatedGradientValues=StaticSwitchParameter(True=Clamp(None=@n95), False=@n94.single) {group=3dBuild, parameterName=3dAnimationBuild}
   Code_BeingDamaged=ScalarParameter() {bUseCustomPrimitiveData=true, group=Build, parameterName=Code_FloorBeingDamaged, primitiveDataIndex=4}
  ) {desc="uv 1, 2 and 3", materialFunction=/Game/Packages/User_Built_Props/Materials/Master/SecondaryMotion}
  n97 = Multiply(A=@n96.Result, B=@n96.Result) {desc=white not built black built}
  n98 = Constant() {r=0.95}
  n99 = Multiply(A=Clamp(None=Subtract(A=@n97, B=@n98)), B=Constant() {r=20})
  n100 = TextureCoordinate() {coordinateIndex=3}
  n101 = Constant() {r=1}
  n102 = If(A=OneMinus(None=ComponentMask(None=@n100) {g=true}), B=@n101, "A > B"=@n101, "A < B"=Constant() {r=-1}) {desc=face boards are white}
  n103 = Multiply(A=Transform(None=Constant3Vector() {constant=<rgba 0 1 0 0>}) {transformSourceType=TRANSFORMSOURCE_Local}, B=@n102) {desc=boards go forward or backward}
  n104 = Multiply(
   A=@n103
   B=Add(A=MaterialFunctionCall() {materialFunction=&SelfBuildingWalls2}.Random Per Mesh) {constB=0.65}
  )
  n105 = VectorParameter() {group=Code, parameterName=CustomExplosionVectorForExplosion}
  n106 = Multiply(
   A=ScalarParameter() {defaultValue=-300, desc=Back and forth values, group=Build, parameterName="outwardMotion(WS)"}
   B=LinearInterpolate(A=StaticSwitchParameter(True=@n104, False=@n103) {group=Build, parameterName=RandomizeDistanceOfLocalLength}, B=@n105.RGB, Alpha=@n105.RGB)
  )
  n107 = MaterialFunctionCall() {materialFunction=&SelfBuildingWalls2}
  n108 = ObjectPositionWS()
  n109 = Multiply(
   A=Normalize(VectorInput=Subtract(A=@n107.Board Pivot Pos, B=@n108))
   B=VectorParameter() {group=Build, parameterName=SphereOffsetAmounts, defaultValue=<rgba 200 200 0 1>}.RGB
  )
  n110 = VectorParameter() {group=Build, parameterName=TargetOffsetLocation, defaultValue=<rgba 0 0 -1024 0>}
  n111 = Add(A=@n109, B=@n110.RGB)
  n112 = Multiply(A=@n97, B=@n96.Result)
  n113 = ScalarParameter() {bUseCustomPrimitiveData=true, group=Build, parameterName=Code_FloorBeingDamaged, primitiveDataIndex=4}
  n114 = AppendVector(A=AppendVector(A=@n97, B=@n97), B=LinearInterpolate(A=@n112, B=@n97, Alpha=@n113))
  n115 = Multiply(A=@n110.RGB, B=@n102) {desc=invert location offset if damage is on}
  n116 = Constant3Vector() {constant=<rgba 0 0 1 0>}
  n117 = VectorParameter() {bUseCustomPrimitiveData=true, group=Code, parameterName=CustomExplosionVectorForExplosion, primitiveDataIndex=16}
  n118 = Add(
   A=LinearInterpolate(A=Add(A=@n109, B=@n115), B=@n111, Alpha=@n113)
   B=LinearInterpolate(A=Multiply(A=@n116, B=@n102) {desc=board up or down}, B=@n117.RGB, Alpha=@n117.RGB)
  ) {desc=z info}
  n119 = Multiply(A=Multiply(A=@n96.Result) {constB=5, desc=20 percent mark}, B=OneMinus(None=@n96.Result))
  n120 = Multiply(A=MaterialFunctionCall() {materialFunction=&SelfBuildingWalls2}.Random Per Mesh) {constB=1.5}
  n121 = Multiply(A=SquareRoot(None=Clamp(None=@n119)), B=Multiply(A=Add(A=@n120), B=Constant3Vector() {desc=up vector amount, constant=<rgba 0 0 100 0>})) {desc=Up animation}
  n122 = StaticSwitchParameter(
   True=Multiply(A=Add(A=@n106, B=@n111), B=@n114)
   False=Add(A=Multiply(A=@n118, B=@n114), B=Multiply(A=@n121, B=@n113) {desc=Only use when floor being damage is on})
  ) {group=Build, parameterName=Walls True Floors False}
  n123 = MaterialFunctionCall() {materialFunction=&SelfBuildingWalls2}
  n124 = Frac(None=Multiply(A=@n123.Board Pivot Pos) {constB=0.3333})
  n125 = Multiply(A=@n96.Result, B=MaterialFunctionCall() {materialFunction=&SelfBuildingWalls2}.Random Per Mesh) {desc=random rotation increases as distance from the start positoin increases}
  n126 = ScalarParameter() {defaultValue=0.5, group=Build, parameterName=Random Axis Rotation Amount}
  n127 = Normalize(VectorInput=LinearInterpolate(A=@n123.Board Orientation, B=ConstantBiasScale(None=@n124) {bias=-0.5, scale=2}, Alpha=Multiply(A=@n125, B=@n126)) {desc=Random to normal axis})
  n128 = DotProduct(
   A=Constant3Vector() {desc=world up vector, constant=<rgba 0 0 1 0>}
   B=MaterialFunctionCall() {materialFunction=&SelfBuildingWalls2}.Board Orientation
  )
  n129 = Add(A=OneMinus(None=Abs(None=@n128)), B=Constant() {r=0.75})
  n130 = LinearInterpolate(A=@n112, B=@n97, Alpha=@n113) {desc=animation linear as wall is damaged sqrd when built}
  n131 = MaterialFunctionCall() {materialFunction=&SelfBuildingWalls2}
  n132 = ScalarParameter() {defaultValue=0.1, group=Build, parameterName=Effect of Random Value on Rotation Amounts}
  n133 = Multiply(
   A=Add(A=Multiply(A=@n131.Random Per Mesh, B=@n132), B=Constant() {r=1})
   B=ScalarParameter() {defaultValue=2, group=Build, parameterName=NumberOfRotations}
  )
  n134 = StaticSwitchParameter(True=Time(), False=Multiply(A=Clamp(None=@n129) {maxDefault=1.25, minDefault=0.5}, B=Multiply(A=@n130, B=@n133) {desc="animation ^2"})) {group=Build, parameterName=DebugRotation}
  n135 = RotateAboutAxis(
   NormalizedRotationAxis=@n127
   RotationAngle=@n134
   PivotPoint=MaterialFunctionCall() {materialFunction=&SelfBuildingWalls2}.Board Pivot Pos
   Position=WorldPosition().XYZ
  )
  n136 = StaticSwitchParameter(True=Add(A=Multiply(A=@n76, B=@n99), B=Add(A=@n122, B=@n135) {desc=rotation + offset}), False=Constant()) {defaultValue=true, group=Build, parameterName=Builder Material}
  n137 = MaterialFunctionCall!(
   -----------Base Shader-----------
   BaseColor1=StaticSwitchParameter(True=@n9, False=StaticSwitchParameter(True=@n52, False=@n61) {group=Build, parameterName=b Damaged but Not Animating Shader}) {desc=d, group=Build, parameterName=Debug vertical opacity mask}
   Base Color Alpha
   SpecularMasks1=@n64
   Emissive=@n71
   Normals1=NamedRerouteUsage()
   World Position Offset=StaticSwitchParameter(True=Constant3Vector(), False=StaticSwitchParameter(True=@n75, False=@n136) {group=Build, parameterName=b Damaged but Not Animating Shader}) {parameterName=MobileVersion}
   DiffuseFresnelAmount
   -----------Destruction------------
   ------Multiple Diffuse Textures-----
   BaseColor2
   BaseColor3
   BaseColor4
   BaseColor5
   BaseColor6
   BaseColor7
   BaseColor8
   " - - - - - - - -"
   SpecularMasks2
   SpecularMasks3
   SpecularMasks4
   SpecularMasks5
   SpecularMasks6
   SpecularMasks7
   SpecularMasks8
   " - - - - - - - - "
   normal2
   Normal3
   Normal4
   Normal5
   Normal6
   Normal7
   Normal8
   " - - - - - - - - 1"
   Use 2 Textures
   Use 3 Textures
   Use 4 Textures
   Use 5 Textures
   Use 6 Textures
   Use 7 Textures
   Use 8 Textures
   Use 9 Textures
   Use 10 Textures
   ExposeLerpFunctionForInvertingTheHitDistanceCalc=StaticBool() {value=true}
  ) {materialFunction=/Game/Packages/Fortress_SharedMaterials/Base_Material_Functions/DEPRECATED_BaseShaderFunction}
  n138 = Frac(None=Multiply(A=ScalarParameter() {defaultValue=0.4095, group=MobileVersion, parameterName=Speed_MobelVersion Debug}, B=Time()))
  n139 = ScalarParameter() {bUseCustomPrimitiveData=true, defaultValue=1, group=Build, parameterName=BuildAnimPhaseNormalized, primitiveDataIndex=2, sliderMax=1}
  n140 = ComponentMask(None=TextureCoordinate() {coordinateIndex=1}) {r=true}
  n141 = Constant() {r=2}
  n142 = TextureSampleParameter2D() {
   group=MobileVersion
   parameterName=BuildAnimMask
   samplerType=SAMPLERTYPE_Masks
   texture=/Game/Packages/PBW/Wood/Textures/Mobile/T_PBW_W1_Mobile_M
  }
  n143 = If(A=@n140, B=Constant() {r=1}, "A > B"=@n142.G, "A == B"=@n142.G, "A < B"=@n142.G)
  n144 = Floor(None=Add(
    A=StaticSwitchParameter(True=@n138, False=@n139) {parameterName=DebugAnim}
    B=If(A=@n140, B=@n141, "A > B"=@n142.B, "A == B"=@n143, "A < B"=@n143)
   ))
  n145 = LinearInterpolate(A=@n1, B=@n137.Diffuse, Alpha=@n144)
  n146 = ScalarParameter() {defaultValue=0.4, group=Banner, parameterName=Dark}
  n147 = ScalarParameter() {defaultValue=1.5, group=Banner, parameterName=Light}
  n148 = CollectionParameter() {collection=&FortniteMaterialParameters, parameterName=BannerPosition}
  n149 = Subtract(A=ComponentMask(None=@n148) {b=true, g=true, r=true}, B=WorldPosition().XYZ)
  n150 = Clamp(None=DotProduct(A=VertexNormalWS(), B=Normalize(VectorInput=@n149)))
  n151 = MaterialFunctionCall!(
   Diffuse Texture Object=TextureObject() {samplerType=SAMPLERTYPE_LinearColor, texture=/Game/Gadgets/Assets/Teleporter/Textures/T_Teleporter_GridSquares_01}
   Spec Texture Object
   TextureSize=ScalarParameter() {defaultValue=300, parameterName=Texture Size}
   --------------------------
   Use seperate settings for Z projection
   Use seperate texture for Z projection
   Texture Size Z projection
   Z Plane Diffuse Texture Object
   Z Plane Spec Texture Object
   "--------------------------      "
   ProjectionTransitionContrast
   Use Texture Alpha Heightmap to Refine the Transistion
   HeightLerpTransitionContrast
   PickTheDiffuseHeightChannel
   TheDiffuseHeightChannel
   "--------------------------       "
   NormalTextureObject
   World Space Normals
   "--------------------------  "
   Export Float 4 Diffuse
   Export Float 4 Spec
   "--------------------------     "
   WorldPosition
   World Space Normal
  ) {materialFunction=/Engine/Functions/Engine_MaterialFunctions02/Texturing/WorldAlignedTextures_Complex}
  n152 = StaticComponentMaskParameter(None=@n151.XYZ Diffuse Texture) {defaultR=true, parameterName=Grid Mask 1}
  n153 = Multiply(A=Time(), B=ScalarParameter() {defaultValue=0.6, group=BG Grid, parameterName=BG Grid Speed})
  n154 = Sine(None=Add(A=ConstantBiasScale(None=@n152) {bias=-0.5, scale=2}, B=Frac(None=@n153)))
  n155 = ScalarParameter() {defaultValue=4, group=BG Grid, parameterName=BG Grid Brightness}
  n156 = Multiply(A=ComponentMask(None=@n151.XYZ Diffuse Texture) {b=true}, B=Clamp(None=Multiply(A=@n154, B=@n155)) {clampMode=CMODE_ClampMin}) {desc=* out the gaps}
  n157 = CollectionParameter() {collection=&FortniteMaterialParameters, parameterName=BannerPosition}
  n158 = ComponentMask(None=@n157) {b=true, g=true, r=true}
  n159 = CollectionParameter() {collection=&FortniteMaterialParameters, parameterName=BannerRadius}
  n160 = CollectionParameter() {collection=&FortniteMaterialParameters, parameterName=BannerTimestamp}
  n161 = MaterialFunctionCall(
   TimestampA=@n160
   TransitionLength_In=ScalarParameter() {defaultValue=2, parameterName=Sweep Length}
   Transition In Only?=StaticBool() {value=true}
  ) {materialFunction=&MF_Timestamp_InOut_01}
  n162 = SphereMask(
   A=WorldPosition().XYZ
   B=@n158
   Radius=LinearInterpolate(B=@n159, Alpha=ConstantBiasScale(None=@n161.Linear) {bias=0, scale=2}) {constAlpha=0.3}
   Hardness=ScalarParameter() {defaultValue=0.4, group=Banner, parameterName=BannerRadiusHardnessSweep}
  )
  n163 = ScalarParameter() {defaultValue=0.6, parameterName=Ring Exponent}
  n164 = OneMinus(None=@n162)
  n165 = ScalarParameter() {defaultValue=8, parameterName=Inverted Ring Exponent}
  n166 = Multiply(
   A=Multiply(A=Power(Base=@n162, Exp=@n163), B=Power(Base=@n164, Exp=@n165))
   B=ScalarParameter() {defaultValue=1.5, parameterName=Ring Brightness}
  )
  n167 = LinearInterpolate(
   A=Multiply(
    A=VectorParameter() {group=Banner, parameterName=Color A, defaultValue=<rgba 0.002 0.015 0.08 1>}.RGB
    B=LinearInterpolate(A=@n146, B=@n147, Alpha=@n150)
   )
   B=VectorParameter() {group=Banner, parameterName=Color B, defaultValue=<rgba 0.5736 0.8246 6 1>}.RGB
   Alpha=Add(A=Multiply(A=@n156, B=@n166), B=@n166)
  )
  n168 = Multiply(A=@n167, B=@n137.Diffuse)
  n169 = SphereMask(
   A=WorldPosition().XYZ
   B=@n158
   Radius=@n159
   Hardness=ScalarParameter() {defaultValue=0.8, group=Banner, parameterName=BannerRadiusHardness}
  )
  n170 = ScalarParameter() {bUseCustomPrimitiveData=true, group=Banner, parameterName=Emissive_Buff_Lv, primitiveDataIndex=15}
  n171 = Multiply(A=Multiply(A=ScalarParameter() {defaultValue=0.25, group=Banner, parameterName=Banner Difuse Color Tint LERP}, B=@n169), B=@n170)
  n172 = VectorParameter() {parameterName=BlueprintColor1, defaultValue=<rgba 0.0885 0.1069 0.125 1>}
  n173 = VectorParameter() {parameterName=BlueprintColor2, defaultValue=<rgba 0.4062 0.77 2 1>}
  n174 = ComponentMask(None=TextureCoordinate() {coordinateIndex=1}) {r=true}
  n175 = Constant() {r=2}
  n176 = TextureSampleParameter2D() {group=Blueprint_Info, parameterName=Blueprint_Outline_Texture, texture=&bluePrintOutlines1}
  n177 = If(A=@n174, B=Constant() {r=1}, "A > B"=@n176.G, "A == B"=@n176.G, "A < B"=@n176.G)
  n178 = ComponentMask(None=ShadingPathSwitch(Default=@n67, Mobile=If(A=@n174, B=@n175, "A > B"=@n176.B, "A == B"=@n177, "A < B"=@n177))) {r=true}
  n179 = StaticSwitchParameter(True=LinearInterpolate(A=LinearInterpolate(A=@n172.RGB, B=@n173.RGB, Alpha=@n178), B=@n1, Alpha=@n144), False=@n1) {parameterName=MobileOpaqueDamaged}
  n180 = NamedRerouteUsage()
  n181 = LinearInterpolate(B=Multiply(A=@n167, B=@n169) {desc=Final Mask out}, Alpha=@n170)
  n182 = MaterialFunctionCall!(In) {materialFunction=/Game/Packages/PBW/BASE/Materials/BASE_Animation}
  n183 = TextureSample(UVs=TextureSample() {texture=/Game/Packages/PBW/BASE/Textures/T_PBW_Base1_01}.R) {texture=/Game/Packages/PBW/BASE/Textures/T_PBW_Base1_G}
  n184 = Distance(
   A=MaterialFunctionCall() {materialFunction=&SelfBuildingWalls2}.Board Pivot Pos
   B=WorldPosition().XYZ
  )
  n185 = Constant() {r=512}
  n186 = StaticSwitchParameter(True=@n137.emissive, False=Constant()) {defaultValue=true, group=Emissive, parameterName=Use_Self-Illumination}
  n187 = StaticSwitchParameter(
   True=VertexColor()
   False=StaticSwitchParameter(True=Divide(A=@n184, B=@n185), False=StaticSwitchParameter(True=@n127, False=@n186) {group=Build, parameterName=DebugRotation}) {group=Build, parameterName=DebugRotationPivotPointLocation}
  ) {group=Build, parameterName=DebugVertexColor}
  n188 = StaticSwitchParameter(True=Add(A=Multiply(A=@n182.BASE_EmissiveOutput, B=@n183.RGB), B=@n187), False=@n187) {group=ConstructorBase, parameterName=ConstructorBASE}
  n189 = TextureSampleParameter2D() {group=Emissive, parameterName=EmissiveTexture, texture=&T_BlacksRGB}
  n190 = VectorParameter() {group=Emissive, parameterName=Emissive Modulate Color, defaultValue=<rgba 1 1 1 1>}
  n191 = Add(
   A=@n188
   B=StaticSwitchParameter(True=Multiply(A=@n189.RGB, B=@n190.RGB), False=@n189.RGB) {group=Emissive, parameterName=Modulate Emissive Texture?}
  )
  n192 = StaticSwitchParameter(
   True=StaticSwitchParameter(True=Multiply(A=@n191, B=@n182.BASE_EmissiveOutput), False=@n191) {group=ConstructorBase, parameterName=Use ConstructorBASE Emissive Animation}
   False=@n188
  ) {group=Emissive, parameterName=Use Emissive Texture}
  n193 = MaterialFunctionCall!(
   BaseColor=StaticSwitchParameter(
    True=StaticSwitchParameter(True=@n145, False=@n137.Diffuse) {parameterName=MobileOpaqueDamaged}
    False=LinearInterpolate(A=@n137.Diffuse, B=@n168, Alpha=@n171)
   ) {parameterName=MobileVersion}
   Emissive=StaticSwitchParameter(True=Add(A=@n179, B=@n180), False=Add(A=@n181, B=@n192)) {parameterName=MobileVersion}
  ) {materialFunction=&/MF_PhasingVisual}
  n194 = MaterialFunctionCall(Float3=@n64) {materialFunction=&BreakOutFloat3Components}
  n195 = OneMinus(None=@n74.Result)
  n196 = Constant() {r=1}
  n197 = StaticSwitchParameter(
   True=Constant() {r=1}
   False=StaticSwitchParameter(True=@n196, False=StaticSwitchParameter(True=@n55, False=@n196) {defaultValue=true, group=Build, parameterName=Builder Material}) {group=Build, parameterName=b Damaged but Not Animating Shader}
  ) {desc=o mask, group=Build, parameterName=Debug vertical opacity mask}
  n198 = ComponentMask(None=NamedRerouteUsage()) {b=true, g=true, r=true}
  n199 = StaticSwitchParameter(
   True=MaterialFunctionCall(
    BaseNormal=MaterialFunctionCall() {materialFunction=&UpdatedNormalAndPixelBuildLevel}.Normals
    AdditionalNormal=@n198
   ) {materialFunction=&BlendAngleCorrectedNormals}.Result
   False=@n198
  ) {defaultValue=true, group=Build, parameterName=Builder Material}
  n200 = Add(
   A=Add(
    A=VectorParameter() {group=Blueprint_Info, parameterName=Blueprint_Outline_Texture_Channel4, defaultValue=<rgba 1 0 0 0>}.RGB
    B=VectorParameter() {group=Blueprint_Info, parameterName=Blueprint_Outline_Texture_Channel3, defaultValue=<rgba 1 0 0 0>}.RGB
   )
   B=VectorParameter() {group=Blueprint_Info, parameterName=Blueprint_Outline_Texture_Channel2, defaultValue=<rgba 1 0 0 0>}.RGB
  )
  n201 = ScalarParameter() {defaultValue=1, group=Blueprint_Info, parameterName=Blueprint_GridMultiplier}
  n202 = Add(
   A=ScalarParameter() {group=Blueprint_Info, parameterName=BluePrint_UseDynamicOrStaticUVs}
   B=ScalarParameter() {defaultValue=1, group=Blueprint_Info, parameterName=Blueprint_Floor_or_Wall}
  )
  n203 = Multiply(A=Add(A=Add(A=@n65.A, B=@n66.A), B=Add(A=@n201, B=@n202)), B=Constant())
  n204 = Constant3Vector() {constant=<rgba 0 0 1 0>}
  n205 = StaticSwitchParameter(
   True=@n204
   False=StaticSwitchParameter(
    True=MaterialFunctionCall(Rotation Axis=@n127, Rotation Angle=@n134) {materialFunction=&FixRotateAboutAxisNormals}.New Tangent Space Vertex Normal
    False=@n204
   ) {defaultValue=true, group=Build, parameterName=Builder Material}
  ) {group=Build, parameterName=b Damaged but Not Animating Shader}
  n206 = OneMinus(None=@n96.Result) {desc=white built}
  n207 = SetMaterialAttributes!(
   MaterialAttributes
   Base Color=MaterialProxyReplace(Realtime=@n193.BaseColor, MaterialProxy=NamedRerouteUsage())
   Metallic=StaticSwitchParameter(True=NamedRerouteUsage(), False=@n194.G) {desc=Metal, parameterName=MobileVersion}
   Specular=StaticSwitchParameter(True=NamedRerouteUsage(), False=@n194.R) {desc=Specular, parameterName=MobileVersion}
   Roughness=StaticSwitchParameter(True=NamedRerouteUsage(), False=@n194.B) {desc=Roughness, parameterName=MobileVersion}
   Emissive Color=MaterialProxyReplace(Realtime=@n193.Emissive, MaterialProxy=NamedRerouteUsage())
   Opacity Mask=StaticSwitchParameter(True=@n144, False=StaticSwitchParameter(True=@n195, False=@n197) {group=Build, parameterName=b Damaged but Not Animating Shader}) {parameterName=MobileVersion}
   Normal=StaticSwitchParameter(
    True=NamedRerouteUsage()
    False=StaticSwitchParameter(True=@n198, False=@n199) {group=Build, parameterName=b Damaged but Not Animating Shader}
   ) {parameterName=MobileVersion}
   World Position Offset=@n137.World Position Offset
   World Displacement=Multiply(B=Add(A=@n200, B=@n203)) {desc=only need to expose to programmers... does not need to work}
   Customized UV 3=ComponentMask(None=@n205) {desc=CustomUV5, g=true, r=true}
   Customized UV 4=AppendVector(A=ComponentMask(None=@n205) {b=true}, B=ConstantBiasScale(None=@n206) {bias=-0.4, scale=1.6667}) {desc=animation phase}
  ) {attributeSetTypes [
    #1
    #2
    #3
    #4
    #5
    #6
    #7
    #8
    #9
    #10
    #11
   ]}
  MaterialFunctionCall(MA In=MaterialFunctionCall!(In=@n207, UseSphericalMappingOverride) {materialFunction=/Game/Packages/Fortress_SharedMaterials/Base_Material_Functions/MF_CurieFallbackEffect}.Result) {materialFunction=&MF_CreativeMoveTool}
  NamedRerouteDeclaration(None=@n33.Specular) {name=TextureLerp_Spec, nodeColor=<rgba 0 1 0.4141 1>}
  NamedRerouteDeclaration(None=@n33.Metalness) {name=TextureLerp_Metal, nodeColor=<rgba 0 0.6484 1 1>}
  NamedRerouteDeclaration(None=@n33.Roughness) {name=TextureLerp_Rough, nodeColor=<rgba 0.8516 1 0 1>}
  NamedRerouteDeclaration(None=@n33.Normals) {name=TextureLerp_Normals, nodeColor=<rgba 0.8281 0 1 1>}
  NamedRerouteDeclaration(None=@n69) {name=CPD Emissive Coolors, nodeColor=<rgba 1 0 0.5156 1>}
  NamedRerouteDeclaration(None=@n52) {name=Diffuse_Out, nodeColor=<rgba 1 0.4219 0 1>}
  NamedRerouteDeclaration(None=@n71) {name=Emissive_Out, nodeColor=<rgba 0 1 0.5781 1>}
  n208 = TextureCoordinate() {uTiling=3, vTiling=3}
  n209 = Constant2Vector() {g=2, r=1}
  TextureSample(UVs=Multiply(
    A=Multiply(A=@n208, B=@n209) {constB=2}
    B=ScalarParameter() {defaultValue=5.3102, group=Build_Mask_Settings, parameterName=RevealMaskTextureTiling}
   )) {samplerType=SAMPLERTYPE_LinearColor, texture=/Game/Global/Textures/Default/Masks_and_Gradients/HexPattern}
 ]
}
