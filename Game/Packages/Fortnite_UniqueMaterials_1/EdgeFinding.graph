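# Translucent placement-preview material that tints an object between good and bad location colors via CanBePlaced.
# Edge detection compares the Diffuse texture's luminance-like channel dot products at four diagonal UV offsets, adding bright outlines to Fresnel-shaded, eye-adaptation-compensated emissive.
# Opacity combines Fresnel and edges with scanlines, plus depth fade on higher deferred quality.
# Vertex offset scales around a world pivot with distance-weighted movement and optional morph targets.
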Material /Game/Packages/Fortnite_UniqueMaterials_1/EdgeFinding
graph {
 properties {
  bEnableResponsiveAA=true
  bUsedWithMeshDeformer=true
  bUsedWithNanite=true
  bUsedWithSkeletalMesh=true
  blendMode=BLEND_Translucent
  pixelDepthOffsetMode=PDOM_Legacy
  translucencyLightingMode=TLM_VolumetricPerVertexDirectional
  usedShadingModels=Default Lit
 }
 nodes [
  Constant() {r=1}
  n1 = VectorParameter() {group=Emissive, parameterName=Placement_Color_Good_Location, defaultValue=<rgba 0.0165 0.0745 0.2788 1>}
  n2 = LinearInterpolate(
   A=VectorParameter() {group=Emissive, parameterName=Placement_Color_Bad_Location, defaultValue=<rgba 0.2788 0.0167 0.0153 1>}.RGB
   B=@n1.RGB
   Alpha=ScalarParameter() {parameterName=CanBePlaced}
  )
  n3 = ScalarParameter() {defaultValue=0.4, parameterName=Desaturation Amount}
  n4 = LinearInterpolate(
   A=ScalarParameter() {defaultValue=0.15, parameterName=Fresnel Low}
   Alpha=MaterialFunctionCall(
    Brightness=ScalarParameter() {defaultValue=1.25, parameterName=Fresnel Brightness}
    Exponent=ScalarParameter() {defaultValue=1, parameterName=Fresnel Exponent}
   ) {materialFunction=&MF_Fresnel_FX}.Fresnel Inverted
  )
  n5 = Constant() {r=1}
  n6 = DotProduct(A=TextureSampleParameter2D() {parameterName=Diffuse, texture=&T_WhitesRGB}.RGB, B=@n5)
  n7 = ScalarParameter() {defaultValue=0.6, parameterName=Contrast Exponent}
  n8 = StaticSwitchParameter(True=Multiply(A=Power(Base=@n6, Exp=@n7), B=ScalarParameter() {defaultValue=3, parameterName=Brightness}), False=Constant() {r=1}) {defaultValue=true, parameterName=Use Outline}
  n9 = ScalarParameter() {defaultValue=6, parameterName=Edge Brightness}
  n10 = TextureCoordinate()
  n11 = ScalarParameter() {defaultValue=0.0001, parameterName=OffsetAmount}
  n12 = DotProduct(
   A=TextureSampleParameter2D(UVs=Add(A=@n10, B=@n11)) {parameterName=Diffuse, texture=&T_WhitesRGB}.RGB
   B=@n5
  )
  n13 = Multiply(A=@n11) {constB=-1}
  n14 = DotProduct(
   A=TextureSampleParameter2D(UVs=Add(A=@n10, B=@n13)) {parameterName=Diffuse, texture=&T_WhitesRGB}.RGB
   B=@n5
  )
  n15 = Add(A=Abs(None=Subtract(A=@n6, B=@n12)), B=Abs(None=Subtract(A=@n6, B=@n14)))
  n16 = AppendVector(A=@n13, B=@n11)
  n17 = DotProduct(
   A=TextureSampleParameter2D(UVs=Add(A=@n10, B=@n16)) {parameterName=Diffuse, texture=&T_WhitesRGB}.RGB
   B=@n5
  )
  n18 = AppendVector(A=@n11, B=@n13)
  n19 = DotProduct(
   A=TextureSampleParameter2D(UVs=Add(A=@n10, B=@n18)) {parameterName=Diffuse, texture=&T_WhitesRGB}.RGB
   B=@n5
  )
  n20 = Add(A=Abs(None=Subtract(A=@n6, B=@n17)), B=Abs(None=Subtract(A=@n6, B=@n19)))
  n21 = Saturate(None=Multiply(A=Add(A=@n15, B=@n20)) {constB=10})
  n22 = StaticSwitchParameter(True=Saturate(None=LinearInterpolate(Alpha=@n21) {constA=-0.9}), False=Constant()) {defaultValue=true, parameterName=Use Outline}
  n23 = Multiply(A=LinearInterpolate(A=@n2, B=Desaturation(None=@n2, Fraction=@n3), Alpha=@n4), B=Add(A=Multiply(A=@n8, B=@n4), B=Multiply(A=@n9, B=@n22)))
  Reroute(None=@n23)
  n24 = Multiply(A=@n23, B=ScalarParameter() {defaultValue=1.2, parameterName=Emissive Brightness}) {constB=0.75}
  n25 = Multiply(A=EyeAdaptation().EyeAdaptation, B=ScalarParameter() {defaultValue=1, parameterName=Eye Adaptation Pre-multiply})
  Reroute(None=LinearInterpolate(A=@n24, B=Divide(A=@n24, B=@n25), Alpha=ScalarParameter() {defaultValue=0.9, parameterName=Eye Adaptation Alpha}))
  n26 = Multiply(A=SquareRoot(None=@n4), B=LinearInterpolate(Alpha=@n22) {constA=0.5})
  n27 = DepthFade(Opacity=@n26) {fadeDistanceDefault=25}
  n28 = ShadingPathSwitch(Default=@n26, Deferred=QualitySwitch(Default=@n26, Low=@n26, High=@n27, Medium=@n27), Mobile=@n26) {desc=Migrated from FeatureLevelSwitch}
  n29 = MaterialFunctionCall!(
   Number of Scanlines
   Scanline Speed
   "Scanlines Thickness (Bias)"=ScalarParameter() {defaultValue=-0.12, parameterName=ScanlineSize}
   Scanlines Dark Value
   Scanlines Light Value
   Block Opacity
   UVs
   Time
   Block Repeat
  ) {materialFunction=/Game/Effects/Fort_Effects/Effects/Atlas/Materials/MF_Atlas_Scanlines_01}
  Reroute(None=Multiply(A=Multiply(A=@n28, B=@n29.Result), B=ScalarParameter() {defaultValue=1, parameterName=Opacity Multiplier}))
  n30 = MaterialFunctionCall!(In=OneMinus(None=ScalarParameter() {group=Morph Targets, parameterName=MorphControl})) {materialFunction=/Game/Effects/Fort_Effects/BluePrints/WeakPointEffect/BlendMorphs}
  n31 = VectorParameter() {parameterName=WorldSpacePivotLocation, defaultValue=<rgba 0 0 0 1>}
  n32 = WorldPosition()
  n33 = StaticSwitchParameter(True=Add(A=@n30.Result, B=@n32.XYZ), False=@n32.XYZ) {group=Morph Targets, parameterName=UseMorphTargets}
  n34 = ScalarParameter() {parameterName=InverseInnerScale}
  n35 = ScalarParameter() {parameterName=InverseOuterScale}
  n36 = Saturate(None=Divide(A=Distance(A=@n31.RGB, B=@n33), B=ScalarParameter() {defaultValue=512, parameterName=ObjectDiameter}))
  n37 = Add(
   A=Multiply(A=Subtract(A=@n31.RGB, B=@n33), B=LinearInterpolate(A=@n34, B=@n35, Alpha=@n36))
   B=LinearInterpolate(B=VectorParameter() {parameterName=InverseMovementVector, defaultValue=<rgba 0 0 0 1>}.RGB, Alpha=Multiply(A=@n36, B=@n36)) {constB=0}
  )
  Reroute(None=StaticSwitchParameter(True=Add(A=@n30.Result, B=@n37), False=@n37) {group=Morph Targets, parameterName=UseMorphTargets})
 ]
}
