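# Unlit translucent, refractive character shield material with fresnel-tinted color, a Hit flash and shatter-style breakup.
# HighlightCracks drives a HeightLerp alpha that samples a break gradient for glow, refraction and opacity, and scales vector-mask chunk displacement.
# Grid-offset sine pulses and scrolling gradient bands along bounding-box height add animated emissive.
# ShadowStanceVisibility alters fresnel, opacity and normal push; IsAthenaShield adds distance-blended fresnel; ShieldMask and depth fade finish opacity.
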Material /Game/Packages/Fortnite_UniqueMaterials_1/Effects/CharacterShield_style1
graph {
 properties {
  bUsedWithMeshDeformer=true
  bUsedWithMorphTargets=true
  bUsedWithNanite=true
  bUsedWithSkeletalMesh=true
  blendMode=BLEND_Translucent
  pixelDepthOffsetMode=PDOM_Legacy
  refractionMethod=RM_IndexOfRefraction
  shadingModel=MSM_Unlit
  usedShadingModels=Unlit
 }
 nodes [
  n1 = VertexNormalWS()
  n2 = ScalarParameter() {defaultValue=0.6372, parameterName=VertPushAmount}
  n3 = Add(A=Multiply(A=ScalarParameter() {parameterName=ShadowStanceVisibility}) {constB=1.5})
  n4 = TextureSample(UVs=TextureCoordinate() {uTiling=0.75, vTiling=0.75}) {samplerType=SAMPLERTYPE_LinearColor, texture=/Game/Packages/Fortnite_UniqueMaterials_1/Textures/ShieldVectorMask}
  n5 = ComponentMask(None=@n4.RGB) {g=true, r=true}
  n6 = OneMinus(None=@n4.G)
  n7 = Normalize(VectorInput=ConstantBiasScale(None=AppendVector(A=@n5, B=@n6)) {bias=-0.5, scale=2})
  n8 = MaterialFunctionCall(
   Transition Phase=Clamp(None=ScalarParameter() {defaultValue=-0.8497, parameterName=HighlightCracks})
   Height Texture=OneMinus(None=Constant())
   Contrast=ScalarParameter() {parameterName=Dissolve Cheap Contrast}
  ) {materialFunction=&HeightLerp}
  n9 = Multiply(
   A=Add(A=VectorParameter() {parameterName=Offset Vector, defaultValue=<rgba 0 0 2 1>}.RGB, B=Transform(None=@n7))
   B=Multiply(A=ScalarParameter() {defaultValue=2, parameterName=SplitChunks}, B=@n8.Alpha)
  )
  n10 = Multiply(A=VertexNormalWS(), B=ScalarParameter() {parameterName=ModelPush})
  Add(
   A=Multiply(A=Normalize(VectorInput=@n1), B=Multiply(A=@n2, B=@n3)) {constB=2}
   B=Multiply(A=Add(A=@n9, B=@n10), B=ScalarParameter() {defaultValue=4, parameterName=Shatter distance})
  )
  n11 = TextureSample(UVs=@n8.Alpha) {desc=R hit falloff - G Pop - B Mask default opacity, texture=/Game/Packages/Fortnite_UniqueMaterials_1/Textures/BreakShieldAnimationGradient}
  n12 = ScalarParameter() {defaultValue=2, parameterName=Refraction death rate}
  LinearInterpolate(B=ScalarParameter() {defaultValue=1.005, parameterName=S Refraction}, Alpha=Clamp(None=Multiply(A=@n11.R, B=@n12) {constB=15})) {constA=1}
  n13 = VectorParameter() {parameterName=S Color1, defaultValue=<rgba 0.994 0.8584 0.4246 1>}
  n14 = VectorParameter() {parameterName=S Color2, defaultValue=<rgba 2 1.2887 0 1>}
  n15 = ScalarParameter() {defaultValue=0.05, parameterName=S Opacity Base Fresnel}
  n16 = LinearInterpolate(
   A=@n15
   B=ScalarParameter() {defaultValue=0.05, parameterName=S Opacity Base Fresnel Far}
   Alpha=MaterialFunctionCall!(
    Blend Range=ScalarParameter() {defaultValue=2048, parameterName=S Opacity Base Fresnel Blend Range}
    Start Offset=ScalarParameter() {defaultValue=2048, parameterName=S Opacity Base Fresnel Blend Start Offset}
   ) {materialFunction=/Engine/Functions/Engine_MaterialFunctions03/Particles/Distance_Blend}.Result
  )
  n17 = ScalarParameter() {parameterName=ShadowStanceVisibility}
  n18 = LinearInterpolate(
   A=Fresnel(
    ExponentIn=ScalarParameter() {defaultValue=1, parameterName=S Opacity Fresnel Power}
    BaseReflectFractionIn=StaticSwitchParameter(True=@n16, False=@n15) {parameterName=IsAthenaShield}
   )
   Alpha=@n17
  ) {constA=1}
  n19 = Fresnel(
   ExponentIn=ScalarParameter() {defaultValue=1, parameterName=S Fresnel Power}
   BaseReflectFractionIn=ScalarParameter() {defaultValue=0.05, parameterName=S Opacity Base Fresnel}
  )
  n20 = ScalarParameter() {parameterName=ShadowStanceVisibility}
  n21 = Multiply(
   A=LinearInterpolate(A=@n13.RGB, B=@n14.RGB, Alpha=@n18)
   B=LinearInterpolate(A=@n13.RGB, B=@n14.RGB, Alpha=LinearInterpolate(A=@n19, Alpha=@n20) {constA=1})
  )
  n22 = Multiply(A=ScalarParameter() {parameterName=Hit}) {constB=12}
  n23 = OneMinus(None=TextureSample(UVs=TextureCoordinate() {uTiling=4, vTiling=4}) {samplerType=SAMPLERTYPE_LinearColor, texture=/Game/Packages/Fortress_SharedMaterials/Textures/T_Slowfield_GridSquares_01}.R)
  n24 = ComponentMask(None=MaterialFunctionCall() {materialFunction=&BoundingBoxBased_0-1_UVW}.RGB) {b=true}
  n25 = Time()
  n26 = MaterialFunctionCall!(Sine Phase=Add(A=Add(A=@n23, B=@n24), B=@n25), Value 1, Value 2) {materialFunction=/Engine/Functions/Engine_MaterialFunctions02/Utility/Sine_Remapped}
  n27 = OneMinus(None=@n26.Remapped Sine)
  n28 = Add(A=Multiply(A=@n11.R, B=@n27), B=Multiply(A=@n11.G, B=@n27))
  n29 = ScalarParameter() {defaultValue=25.7622, parameterName=Dissolve Pattern Emissive Brightness}
  n30 = OneMinus(None=@n24)
  n31 = TextureSample!(UVs=Frac(None=Add(A=@n30, B=@n25)), Tex, Level, Apply View MipBias) {
   desc=R hit falloff - G Pop - B Mask default opacity
   mipValueMode=TMVM_MipLevel
   texture=&BreakShieldAnimationGradient
  }
  n32 = Multiply(A=@n25) {constB=0.5, desc=UVS}
  n33 = Multiply(A=@n24) {constB=2, desc=UVS}
  n34 = TextureSample(UVs=Frac(None=Add(A=@n32, B=@n33))) {
   desc=R hit falloff - G Pop - B Mask default opacity
   mipValueMode=TMVM_MipLevel
   texture=&BreakShieldAnimationGradient
  }
  n35 = Add(A=@n33, B=Multiply(A=@n25) {constB=0.2758})
  n36 = TextureSample(UVs=Frac(None=Add(A=@n35) {constB=0.5477})) {
   desc=R hit falloff - G Pop - B Mask default opacity
   mipValueMode=TMVM_MipLevel
   texture=&BreakShieldAnimationGradient
  }
  n37 = Add(A=Multiply(A=@n31.R) {constB=0.25}, B=Add(A=Multiply(A=@n34.G) {constB=0.1}, B=Multiply(A=@n36.R) {constB=0.1376}))
  n38 = StaticSwitchParameter(True=LinearInterpolate(Alpha=@n37) {constA=0.25, constB=0.6}, False=@n37) {parameterName=IsAthenaShield}
  n39 = Multiply(A=@n38, B=@n26.Remapped Sine)
  MaterialFunctionCall!(
   Base=MaterialFunctionCall(
    Base=MaterialFunctionCall(Base=@n21, Add=@n22) {materialFunction=/Engine/Functions/Engine_MaterialFunctions02/Math/MultiplyAdd}.Result
    Add=Multiply(A=@n28, B=@n29) {constB=60.844}
   ) {materialFunction=&MultiplyAdd}.Result
   Add=Multiply(A=Add(A=@n39, B=@n38)) {constB=10}
  ) {materialFunction=&MultiplyAdd}
  n40 = TextureSampleParameter2D() {
   parameterName=ShieldMask
   samplerType=SAMPLERTYPE_Masks
   texture=/Game/Characters/Player/Female/Medium/Bodies/F_MED_Commando_01/Textures/SolidWhite_MaskCompression
  }
  n41 = ScalarParameter() {defaultValue=0.1081, parameterName=S Opacity}
  n42 = Multiply(
   A=Multiply(A=LinearInterpolate(Alpha=@n17) {constA=1, constB=0.02}, B=Multiply(A=@n18, B=@n41))
   B=ScalarParameter() {defaultValue=1, parameterName=Opacity}
  )
  n43 = OneMinus(None=Multiply(A=OneMinus(None=@n4.B), B=@n8.Alpha))
  n44 = Add(A=Multiply(A=Multiply(A=@n42, B=@n43), B=@n11.B), B=Multiply(A=@n28) {constB=0.2})
  Multiply(
   A=@n40.RGB
   B=DepthFade(Opacity=Clamp(None=@n44), FadeDistance=ScalarParameter() {defaultValue=1.4, parameterName=Fade Distance}) {fadeDistanceDefault=10}
  )
 ]
}
