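# DroidScreenPosition is a material function that outputs a world-space position defined relative to the camera, presumably to place a droid effect on screen.
# The CameraOffset vector parameter supplies the offset, with red and green each multiplied by themselves and blue passed through as Z.
# MakeFloat3 combines these, and TransformPosition converts the result from camera to world space.
# Because of the squaring, X and Y lose their sign and cannot be negative.
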MaterialFunction /Game/Packages/Fortnite_UniqueMaterials_1/Effects/DroidScreenPosition
graph {nodes [
  n1 = VectorParameter() {parameterName=CameraOffset, defaultValue=<rgba 0.5 -0.5 11 1>}
  n2 = Multiply(A=@n1.R, B=@n1.R)
  n3 = Multiply(A=@n1.G, B=@n1.G)
  FunctionOutput(None=TransformPosition(None=MaterialFunctionCall(X=@n2, Y=@n3, Z=@n1.B) {materialFunction=&MakeFloat3}.Result) {transformSourceType=TRANSFORMPOSSOURCE_Camera, transformType=TRANSFORMPOSSOURCE_World})
 ]}
