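# Unlit material MGreen for what parameter names suggest are stair arrow, X and blinking-sign indicators.
# Packed texture channels blend BlueColor toward gray, desaturated by isX or isblinkingSign.
# Those flags also select between opacity calculations using InteractParameter, with a static switch enabling a LinearSine arrow pulse.
# A separate chain multiplies scaled CustomDepth by a DepthFade intersection blend; no material output connections appear, and SinePeriodPulseSpeed is unused.
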Material /Game/Packages/Fortnite_UniqueMaterials_1/Effects/MGreen
graph {
 properties {
  bUseTranslucencyVertexFog=false
  pixelDepthOffsetMode=PDOM_Legacy
  shadingModel=MSM_Unlit
  translucencyLightingMode=TLM_VolumetricPerVertexNonDirectional
  usedShadingModels=Unlit
 }
 nodes [
  n1 = TextureSampleParameter2D() {parameterName=Diffuse, texture=/Game/Packages/DS_Fortnite_Mesh/BlueprintTest1}
  n2 = Add(A=@n1.R, B=Multiply(A=@n1.G) {constB=0.5})
  n3 = Saturate(None=Multiply(A=@n2, B=@n2))
  n4 = LinearInterpolate(
   A=VectorParameter() {parameterName=BlueColor, defaultValue=<rgba 0.0129 0.0896 0.29 1>}.RGB
   B=Constant() {r=0.5}
   Alpha=StaticSwitchParameter(True=Multiply(A=@n3) {constB=0.5}, False=Multiply(A=@n1.R, B=@n1.R)) {group=Stairs, parameterName=UseDirectionalArrowPulse}
  )
  n5 = Clamp(None=Add(A=ScalarParameter() {parameterName=isX}, B=ScalarParameter() {parameterName=isblinkingSign}))
  n6 = Desaturation(None=StaticSwitchParameter(True=Saturate(None=@n4), False=@n4) {group=Stairs, parameterName=UseDirectionalArrowPulse}, Fraction=@n5)
  n7 = Add(A=@n1.B, B=Constant() {r=0.5})
  n8 = Constant() {desc=1.5, r=1.5}
  n9 = ScalarParameter() {defaultValue=0.75, parameterName=Opacity}
  n10 = ScalarParameter() {parameterName=InteractParameter}
  n11 = Clamp(None=Multiply(A=Divide(A=@n7, B=@n8), B=Add(A=@n9, B=@n10))) {minDefault=0.75}
  n12 = ScalarParameter() {defaultValue=1, group=Stairs, parameterName=sineOffset}
  n13 = MaterialFunctionCall(Value=Time()) {materialFunction=&LinearSine}
  n14 = LinearInterpolate(Alpha=Multiply(A=Add(A=@n12, B=@n13.Linear Sine), B=@n1.G)) {constB=0.75}
  n15 = ScalarParameter() {parameterName=isblinkingSign}
  n16 = Add(A=Add(A=@n1.B, B=Multiply(A=@n14, B=@n15)), B=Multiply(A=@n15) {constB=0.25})
  n17 = OneMinus(None=Multiply(B=ScalarParameter() {parameterName=isX}) {constA=0.75, desc=inverse of desired x square opacity})
  n18 = StaticSwitchParameter(True=Clamp(None=Add(A=@n10, B=@n16)), False=Add(A=@n1.B, B=Multiply(A=@n10, B=@n17)) {constB=0}) {group=Stairs, parameterName=UseDirectionalArrowPulse}
  Add(A=@n6, B=Multiply(A=LinearInterpolate(A=@n11, B=@n18, Alpha=@n5), B=ScalarParameter() {defaultValue=1, parameterName=FoRealzOpacity}))
  Constant() {r=1}
  Reroute(None=Multiply(A=@n6, B=Constant() {r=2}) {constB=2})
  n19 = Saturate(None=Divide(A=SceneTexture() {sceneTextureId=PPI_CustomDepth}.Color, B=ScalarParameter() {defaultValue=50000, parameterName=CustomDepthDivisor}))
  n20 = ScalarParameter() {defaultValue=0.4, parameterName=Opacity-Intersecting}
  n21 = ScalarParameter() {defaultValue=1, parameterName=OpacityB}
  n22 = DepthFade(Opacity=Constant() {r=1}, FadeDistance=ScalarParameter() {defaultValue=5, parameterName=FadeDistance})
  Reroute(None=Multiply(A=ComponentMask(None=@n19) {r=true}, B=LinearInterpolate(A=@n20, B=@n21, Alpha=@n22)))
  ScalarParameter() {defaultValue=1, group=Stairs, parameterName=SinePeriodPulseSpeed}
  StaticBool()
 ]
}
