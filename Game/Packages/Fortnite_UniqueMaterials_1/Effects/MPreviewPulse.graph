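# MPreviewPulse is a translucent, default-lit material for a pulsing, color-shifting effect, by its name for previews.
# Time plus a world-normal Fresnel forms the UV for a vertical gradient mask, which blends two blue-default emissive color parameters into a scrolling view-dependent pulse.
# BaseColor, Spec and Rough parameters sit alongside it.
# Sine oscillators and constant colors are also present, with no connections shown from them or to material outputs.
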Material /Game/Packages/Fortnite_UniqueMaterials_1/Effects/MPreviewPulse
graph {
 properties {
  blendMode=BLEND_Translucent
  pixelDepthOffsetMode=PDOM_Legacy
  translucencyLightingMode=TLM_VolumetricPerVertexNonDirectional
  usedShadingModels=Default Lit
 }
 nodes [
  Constant3Vector() {constant=<rgba 0 0.0319 0.1 1>}
  Constant3Vector() {constant=<rgba 0 0.0639 0.2 1>}
  Constant() {r=0.1}
  n1 = Divide(A=Time(), B=Constant() {r=2})
  Divide(A=Add(A=Sine(None=@n1), B=Constant() {r=1.5}), B=Constant() {r=4})
  n2 = Sine(None=Multiply(A=Time(), B=Constant() {r=0.5}))
  LinearInterpolate(A=Constant() {r=0.75}, B=Constant() {r=0.85}, Alpha=ConstantBiasScale(None=@n2))
  Multiply(A=Add(A=@n2, B=Constant() {r=1}), B=Constant() {r=0.25})
  Reroute(None=VectorParameter() {parameterName=BaseColor, defaultValue=<rgba 0 0 0 1>}.RGB)
  n3 = Add(A=Multiply(A=Time(), B=Constant() {r=0.25}), B=Fresnel(ExponentIn=Constant() {r=1}, Normal=VertexNormalWS()) {baseReflectFraction=0.01})
  Reroute(None=LinearInterpolate(
    A=VectorParameter() {parameterName=EmissiveColor, defaultValue=<rgba 0 0.0639 0.2 1>}.RGB
    B=VectorParameter() {parameterName=EmissiveColor2, defaultValue=<rgba 0 0.0319 0.1 1>}.RGB
    Alpha=TextureSample(UVs=@n3) {samplerType=SAMPLERTYPE_LinearColor, texture=/Game/Global/Textures/Default/Masks_and_Gradients/T_VertGradient}.R
   ))
  Reroute(None=ScalarParameter() {defaultValue=0.3, parameterName=Spec})
  Reroute(None=ScalarParameter() {defaultValue=0.5, parameterName=Rough})
 ]
}
