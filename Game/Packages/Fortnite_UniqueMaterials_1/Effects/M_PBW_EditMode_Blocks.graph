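# Unlit translucent material for blocks in building edit mode, drawn without depth testing.
# A blueprint texture blends a blue tint toward grey, with a static switch for a stairs arrow variant, then gets brightened.
# The isX and isblinkingSign parameters desaturate that color.
# Opacity multiplies normalized CustomDepth by a DepthFade blend that lowers alpha where blocks intersect scene geometry.
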Material /Game/Packages/Fortnite_UniqueMaterials_1/Effects/M_PBW_EditMode_Blocks
graph {
 properties {
  bDisableDepthTest=true
  bUseTranslucencyVertexFog=false
  blendMode=BLEND_Translucent
  pixelDepthOffsetMode=PDOM_Legacy
  shadingModel=MSM_Unlit
  translucencyLightingMode=TLM_VolumetricPerVertexNonDirectional
  usedShadingModels=Unlit
 }
 nodes [
  n1 = Divide(A=SceneTexture() {sceneTextureId=PPI_CustomDepth}.Color, B=ScalarParameter() {defaultValue=50000, parameterName=CustomDepthDivisor})
  n2 = Constant() {r=1}
  n3 = ScalarParameter() {defaultValue=5, parameterName=FadeDistance}
  Multiply(
   A=ComponentMask(None=Saturate(None=@n1)) {r=true}
   B=LinearInterpolate(
    A=ScalarParameter() {defaultValue=0.4, parameterName=Opacity-Intersecting}
    B=ScalarParameter() {defaultValue=1, parameterName=OpacityB}
    Alpha=DepthFade(Opacity=@n2, FadeDistance=@n3)
   )
  )
  n4 = TextureSampleParameter2D() {parameterName=Diffuse, texture=/Game/Packages/DS_Fortnite_Mesh/BlueprintTest1}
  n5 = Add(A=@n4.R, B=Multiply(A=@n4.G) {constB=0.5})
  n6 = Saturate(None=Multiply(A=@n5, B=@n5))
  n7 = LinearInterpolate(
   A=VectorParameter() {parameterName=BlueColor, defaultValue=<rgba 0.0129 0.0896 0.29 1>}.RGB
   B=Constant() {r=0.5}
   Alpha=StaticSwitchParameter(True=Multiply(A=@n6) {constB=0.5}, False=Multiply(A=@n4.R, B=@n4.R)) {group=Stairs, parameterName=UseDirectionalArrowPulse}
  )
  n8 = Saturate(None=@n7)
  n9 = Add(A=ScalarParameter() {parameterName=isX}, B=ScalarParameter() {parameterName=isblinkingSign})
  Multiply(
   A=Desaturation(None=StaticSwitchParameter(True=@n8, False=@n7) {group=Stairs, parameterName=UseDirectionalArrowPulse}, Fraction=Clamp(None=@n9))
   B=Constant() {r=2}
  ) {constB=2}
 ]
}
