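# Tree material function that declares the scalar wind parameters which, per its naming, code references and animates.
# Parameters span four Wind Setting groups, covering radius, rotation, offset, shelter, phase offset and branch mask, with some marked Do Not Animate.
# An Add tree sums them into a dummy BaseColor, used only when the Display Code Referenced Variables static switch is on.
# With the switch off, as by default, input material attributes pass through unchanged.
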MaterialFunction /Game/Packages/Fortnite_UniqueMaterials_1/MaterialFunctions/Trees/MF_CodeReferenceAnimatedVars
graph {nodes [
  n1 = Add(
   A=ScalarParameter() {defaultValue=1, group=Wind Setting 1, parameterName=Radius Multiplier 1}
   B=ScalarParameter() {defaultValue=0.0092, group=Wind Setting 1, parameterName=Max Rotation 1, sliderMax=10}
  )
  n2 = Add(A=ScalarParameter() {group=Wind Setting 1, parameterName=Wind offset 1})
  n3 = Add(B=ScalarParameter() {defaultValue=250, group=Wind Setting 1, parameterName=Wind 1 Random Rotation Change Rate})
  n4 = Add(
   A=ScalarParameter() {defaultValue=0.25, group=Wind Setting 1, parameterName=Wind 1 Random Rotation Influence}
   B=ScalarParameter() {defaultValue=-1, group=Wind Setting 1, parameterName=Wind 1 Shelter Bias}
  )
  n5 = Add(
   A=Add(A=Add(A=@n1, B=@n2), B=Add(A=@n3, B=@n4))
   B=Add(
    A=ScalarParameter() {defaultValue=-2, group=Wind Setting 1, parameterName=Wind 1 Shelter Scale}
    B=ScalarParameter() {defaultValue=-2000, group=Wind Setting 1, parameterName=Wind 1 - Vertex Phase Offset X Axis}
   )
  )
  n6 = Add(B=ScalarParameter() {defaultValue=0.01, group=Wind Setting 2, parameterName=Wind 1 - Harsh Root-Based Branch Length Animation Mask})
  n7 = ScalarParameter() {defaultValue=1, group=Wind Setting 2, parameterName=Radius Multiplier 2}
  n8 = ScalarParameter() {defaultValue=0.0092, group=Wind Setting 2, parameterName=Max Rotation 2, sliderMax=10}
  n9 = ScalarParameter() {group=Wind Setting 2, parameterName=Wind offset 2}
  n10 = ScalarParameter() {defaultValue=1, group=Wind Setting 2, parameterName=Wind 2 Speed - Do Not Animate}
  n11 = ScalarParameter() {defaultValue=250, group=Wind Setting 2, parameterName=Wind 2 Random Rotation Change Rate}
  n12 = ScalarParameter() {defaultValue=0.25, group=Wind Setting 2, parameterName=Wind 2 Random Rotation Influence}
  n13 = ScalarParameter() {defaultValue=-1, group=Wind Setting 2, parameterName=Wind 2 Shelter Bias}
  n14 = ScalarParameter() {defaultValue=-2, group=Wind Setting 2, parameterName=Wind 2 Shelter Scale}
  n15 = Add(A=Add(A=Add(A=@n7, B=@n8), B=Add(A=@n9, B=@n10)), B=Add(A=Add(A=@n11, B=@n12), B=Add(A=@n13, B=@n14)))
  n16 = ScalarParameter() {defaultValue=0.0092, group=Wind Setting 3, parameterName=Max Rotation 3, sliderMax=10}
  n17 = ScalarParameter() {group=Wind Setting 3, parameterName=Wind offset 3}
  n18 = Add(
   A=Add(
    A=ScalarParameter() {defaultValue=-2000, group=Wind Setting 2, parameterName=Wind 2 - Vertex Phase Offset X Axis}
    B=ScalarParameter() {defaultValue=0.01, group=Wind Setting 2, parameterName=Wind 2 - Harsh Root-Based Branch Length Animation Mask}
   )
   B=Add(A=ScalarParameter() {defaultValue=1, group=Wind Setting 3, parameterName=Radius Multiplier 3}, B=Add(A=@n16, B=@n17))
  )
  n19 = Add(
   A=ScalarParameter() {defaultValue=0.25, group=Wind Setting 3, parameterName=Wind 3 Random Rotation Influence}
   B=ScalarParameter() {defaultValue=-1, group=Wind Setting 3, parameterName=Wind 3 Shelter Bias}
  )
  n20 = Add(
   A=ScalarParameter() {defaultValue=-2, group=Wind Setting 3, parameterName=Wind 3 Shelter Scale}
   B=ScalarParameter() {defaultValue=-2000, group=Wind Setting 3, parameterName=Wind 3 - Vertex Phase Offset X Axis}
  )
  n21 = Add(
   A=Add(
    A=ScalarParameter() {defaultValue=1, group=Wind Setting 3, parameterName=Wind 3 Speed - Do Not Animate}
    B=ScalarParameter() {defaultValue=250, group=Wind Setting 3, parameterName=Wind 3 Random Rotation Change Rate - Do Not Animate}
   )
   B=Add(
    A=Add(A=@n19, B=@n20)
    B=ScalarParameter() {defaultValue=0.01, group=Wind Setting 3, parameterName=Wind 3 - Harsh Root-Based Branch Length Animation Mask}
   )
  )
  n22 = ScalarParameter() {group=Wind Setting 4, parameterName=Radius Multiplier 4}
  n23 = ScalarParameter() {defaultValue=0.0092, group=Wind Setting 4, parameterName=Max Rotation 4, sliderMax=10}
  n24 = ScalarParameter() {group=Wind Setting 4, parameterName=Wind offset 4}
  n25 = Add(A=Add(B=Add(A=@n22, B=@n23)), B=Add(A=Add(A=@n24)))
  n26 = Add(
   A=Add(
    A=ScalarParameter() {defaultValue=250, group=Wind Setting 4, parameterName=Wind 4 Random Rotation Change Rate}
    B=ScalarParameter() {defaultValue=0.25, group=Wind Setting 4, parameterName=Wind 4 Random Rotation Influence}
   )
   B=Add(
    A=ScalarParameter() {defaultValue=-1, group=Wind Setting 4, parameterName=Wind 4 Shelter Bias}
    B=ScalarParameter() {defaultValue=-2, group=Wind Setting 4, parameterName=Wind 4 Shelter Scale}
   )
  )
  n27 = Add(
   A=ScalarParameter() {defaultValue=-2000, group=Wind Setting 4, parameterName=Wind 4 - Vertex Phase Offset X Axis}
   B=ScalarParameter() {defaultValue=0.01, group=Wind Setting 4, parameterName=Wind 4 - Harsh Root-Based Branch Length Animation Mask}
  )
  n28 = Add(A=Add(A=Add(A=@n5, B=@n6), B=Add(A=@n15, B=@n18)), B=Add(A=Add(A=@n21, B=@n25), B=Add(A=@n26, B=@n27)))
  FunctionOutput(None=StaticSwitchParameter(
    True=MakeMaterialAttributes!(
     BaseColor=@n28
     Metallic
     Specular
     Roughness
     Anisotropy
     EmissiveColor
     Opacity
     OpacityMask
     Normal
     Tangent
     WorldPositionOffset
     SubsurfaceColor
     ClearCoat
     ClearCoatRoughness
     AmbientOcclusion
     Refraction
     CustomizedUVs_0
     CustomizedUVs_1
     CustomizedUVs_2
     CustomizedUVs_3
     CustomizedUVs_4
     CustomizedUVs_5
     CustomizedUVs_6
     CustomizedUVs_7
     PixelDepthOffset
     ShadingModel
     Displacement
    )
    False=FunctionInput() {inputType=FunctionInput_MaterialAttributes}
   ) {group=Code, parameterName=Display Code Referenced Variables})
 ]}
