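# Material function for trees that optionally tints a canopy's base color with a vertical top-to-bottom color gradient.
# It lerps two color parameters by the green channel of UV channel 2, then fades that tint from white using a Mask input.
# The tint multiplies the base color extracted from the input material attributes and is written back through MatLayerBlend_OverrideBaseColor.
# A static switch passes the material through unchanged when the gradient is off.
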MaterialFunction /Game/Packages/Fortnite_UniqueMaterials_1/MaterialFunctions/Trees/MF_TreeColorGradient
graph {nodes [
  n1 = FunctionInput() {inputType=FunctionInput_MaterialAttributes}
  n2 = MaterialFunctionCall!(Material=@n1) {materialFunction=/Engine/Functions/MaterialLayerFunctions/MatLayerBlend_BreakBaseColor}
  n3 = TextureCoordinate() {coordinateIndex=2}
  n4 = FunctionInput() {bUsePreviewValueAsDefault=true, inputName=Mask, inputType=FunctionInput_Scalar, sortPriority=3, previewValue=<1 0 0 1>}
  n5 = LinearInterpolate(
   A=Constant() {r=1}
   B=LinearInterpolate(
    A=VectorParameter() {group=ColorGradient, parameterName=Gradient_TopColor, defaultValue=<rgba 1 1 1 1>}.RGB
    B=VectorParameter() {group=ColorGradient, parameterName=Gradient_BottomColor, defaultValue=<rgba 1 1 1 1>}.RGB
    Alpha=MaterialFunctionCall(Float2=@n3) {materialFunction=&BreakOutFloat2Components}.G
   )
   Alpha=@n4
  )
  n6 = StaticSwitchParameter(
   True=MaterialFunctionCall!(Material=@n1, NewBaseColor=Multiply(A=@n2.BaseColor, B=@n5), Mask) {materialFunction=/Engine/Functions/MaterialLayerFunctions/MatLayerBlend_OverrideBaseColor}.Blended Material
   False=@n1
  ) {
   desc=The gradient uses the green channel of the 3rd uv channel from max.
   group=ColorGradient
   parameterName=Apply UV 2.g Canopy Diffuse Color Gradient
  }
  FunctionOutput(None=@n6)
 ]}
