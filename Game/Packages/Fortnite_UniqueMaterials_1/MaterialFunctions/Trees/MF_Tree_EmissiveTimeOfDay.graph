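# Tree material function that gives a Material Attributes input an emissive glow whose color follows time of day, behind the UseEmissiveTimeOfDay static switch.
# A selectable channel of a mask texture multiplies a color from Color_Time_of_Day, fed by morning, day, evening and night parameters.
# An inverted CameraDepthFade fades the glow with distance before MatLayerBlend_Emissive applies it; with the switch off, the input passes through unchanged.
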MaterialFunction /Game/Packages/Fortnite_UniqueMaterials_1/MaterialFunctions/Trees/MF_Tree_EmissiveTimeOfDay
graph {nodes [
  n1 = TextureSampleParameter2D() {
   group=EmissiveTimeOfDay
   parameterName=EmissiveTODTexture
   samplerType=SAMPLERTYPE_Masks
   texture=&T_Fortnite_Default_S
  }
  n2 = VectorParameter() {group=EmissiveTimeOfDay, parameterName=ColorEmissive_Morning, defaultValue=<rgba 1 1 1 1>}
  n3 = VectorParameter() {group=EmissiveTimeOfDay, parameterName=ColorEmissive_Day, defaultValue=<rgba 1 1 1 1>}
  n4 = VectorParameter() {group=EmissiveTimeOfDay, parameterName=ColorEmissive_Evening, defaultValue=<rgba 1 1 1 1>}
  n5 = VectorParameter() {group=EmissiveTimeOfDay, parameterName=ColorEmissive_Night, defaultValue=<rgba 1 1 1 1>}
  n6 = ScalarParameter() {defaultValue=5000, group=EmissiveTimeOfDay, parameterName=EmissiveDistanceFade}
  n7 = ScalarParameter() {defaultValue=2000, group=EmissiveTimeOfDay, parameterName=EmissiveDistanceFade_Offset}
  n8 = Multiply(
   A=Multiply(
    A=StaticComponentMaskParameter(None=@n1.RGB) {defaultR=true, group=EmissiveTimeOfDay, parameterName=EmissiveTOD_ComponentMask}
    B=MaterialFunctionCall(Color_Morning=@n2.RGB, Color_Day=@n3.RGB, Color_Evening=@n4.RGB, Color_Night=@n5.RGB) {materialFunction=&Color_Time_of_Day}.Result
   )
   B=OneMinus(None=MaterialFunctionCall(Fade Length=@n6, Fade Offset=@n7) {materialFunction=&CameraDepthFade}.Result)
  )
  n9 = FunctionInput() {inputType=FunctionInput_MaterialAttributes}
  FunctionOutput(None=StaticSwitchParameter(
    True=MaterialFunctionCall!(Material=@n9, Emissive=@n8) {materialFunction=/Engine/Functions/MaterialLayerFunctions/MatLayerBlend_Emissive}.Blended Material
    False=@n9
   ) {group=EmissiveTimeOfDay, parameterName=UseEmissiveTimeOfDay})
 ]}
