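# Material function for tree materials that adds a fuzzy-shading fresnel effect to the base color of incoming material attributes.
# It breaks out base color, runs world-space fuzzy shading with core darkness, core power and edge brightness parameters, then writes the result back via an override blend.
# Low and Medium quality skip the effect through a quality switch, and material proxy baking uses the unmodified base color.
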MaterialFunction /Game/Packages/Fortnite_UniqueMaterials_1/MaterialFunctions/Trees/MF_Tree_Fresnel
graph {nodes [
  n1 = FunctionInput() {inputType=FunctionInput_MaterialAttributes}
  n2 = MaterialFunctionCall!(Material=@n1) {materialFunction=/Engine/Functions/MaterialLayerFunctions/MatLayerBlend_BreakBaseColor}
  n3 = QualitySwitch(
   Default=MaterialFunctionCall!(
    BaseColor=@n2.BaseColor
    Normal
    CoreDarkness=ScalarParameter() {group=PX, parameterName=Fresnel Core Darkness}
    Power=ScalarParameter() {defaultValue=3, group=PX, parameterName=Fresnel Core Power}
    EdgeBrightness=ScalarParameter() {defaultValue=1, group=PX, parameterName=Fresnel Edge Brightness}
   ) {materialFunction=/Game/Packages/Fortress_SharedMaterials/Base_Material_Functions/FuzzyShadingWorldSpace}.Result
   Low=@n2.BaseColor
   Medium=@n2.BaseColor
  )
  FunctionOutput(None=MaterialFunctionCall!(Material=@n1, NewBaseColor=MaterialProxyReplace(Realtime=@n3, MaterialProxy=@n2.BaseColor), Mask) {materialFunction=/Engine/Functions/MaterialLayerFunctions/MatLayerBlend_OverrideBaseColor}.Blended Material)
 ]}
