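# Master material for Fortnite trees and foliage, using two-sided foliage shading with Nanite and instancing support.
# Packed mask channels (opacity, leaves, damage) feed a material-attribute chain of tree functions for animation, color variation, damage, subsurface, snow and more.
# World position offset adds destructible damage motion and an optional camera-direction push tied to pixel depth offset.
# Final stages add the creative move tool and switchable lava edge, Curie fallback and progressive snow effects.
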Material /Game/Packages/Fortnite_UniqueMaterials_1/FortniteFoliage_MasterMaterial
graph {
 properties {
  bTangentSpaceNormal=false
  bUseMaterialAttributes=true
  bUsedWithInstancedStaticMeshes=true
  bUsedWithNanite=true
  ditheredLODTransition=true
  numCustomizedUVs=5
  physMaterial=/Game/FortressPhysicalMaterials/Grass
  pixelDepthOffsetMode=PDOM_Legacy
  refractionMethod=RM_IndexOfRefraction
  shadingModel=MSM_TwoSidedFoliage
  usedShadingModels=Two Sided Foliage
 }
 nodes [
  n1 = TextureSampleParameter2D() {
   group=Textures
   parameterName=Diffuse
   samplerSource=SSM_Wrap_WorldGroupSettings
   texture=/Game/Environments/Props/Trees/Textures/HeroTreesOPT_D
  }
  n2 = ComponentMask(None=@n1.RGB) {b=true, desc=BaseColor, g=true, r=true}
  n3 = Constant()
  n4 = TextureSampleParameter2D() {
   desc="spec, metallic,roughness\n"
   group=Specular
   parameterName=SpecularMasks
   samplerSource=SSM_Wrap_WorldGroupSettings
   samplerType=SAMPLERTYPE_Masks
   texture=&T_Fortnite_Default_S
  }
  n5 = MaterialFunctionCall(Float3=@n4.RGB) {desc=SpecularTexutre, materialFunction=&BreakOutFloat3Components}
  n6 = TextureSampleParameter2D() {
   group=PX
   parameterName=Normals
   samplerSource=SSM_Wrap_WorldGroupSettings
   samplerType=SAMPLERTYPE_Normal
   texture=/Game/Environments/Props/Trees/Textures/HeroTreesOPT_N
  }
  n7 = ComponentMask(None=@n6.RGB) {b=true, desc=Normal, g=true, r=true}
  n8 = TextureSampleParameter2D() {
   group=Textures
   parameterName=MaskTexture
   samplerSource=SSM_Wrap_WorldGroupSettings
   samplerType=SAMPLERTYPE_Masks
   texture=&T_WhiteMask
  }
  n9 = MaterialFunctionCall(Float3=@n8.RGB) {desc=Mask Texture, materialFunction=&BreakOutFloat3Components}
  n10 = MaterialFunctionCall!(In=MaterialFunctionCall!(
    In=MakeMaterialAttributes!(
     BaseColor=@n2
     Metallic=@n3
     Specular=@n5.R
     Roughness=@n5.R
     Anisotropy
     EmissiveColor
     Opacity
     OpacityMask=@n9.R
     Normal=@n7
     Tangent
     WorldPositionOffset
     SubsurfaceColor
     ClearCoat
     ClearCoatRoughness
     AmbientOcclusion
     Refraction
     CustomizedUVs_0
     CustomizedUVs_1
     CustomizedUVs_2
     CustomizedUVs_3
     CustomizedUVs_4
     CustomizedUVs_5
     CustomizedUVs_6
     CustomizedUVs_7
     PixelDepthOffset
     ShadingModel
     Displacement
    )
    Mask=@n9.G
   ) {materialFunction=/Game/Packages/Fortnite_UniqueMaterials_1/MaterialFunctions/Trees/MF_TreePanningNormals}.Result) {materialFunction=/Game/Packages/Fortnite_UniqueMaterials_1/MaterialFunctions/Trees/MF_4Level_TreeAnimations}
  n11 = MaterialFunctionCall!(
   In=MaterialFunctionCall!(
    In=MaterialFunctionCall!(In=@n10.Result, Position and Parent Index=@n10.Result, Mesh Element Index=@n10.Result) {materialFunction=/Game/Packages/Fortnite_UniqueMaterials_1/MaterialFunctions/Trees/MF_TreeRandomColorElements}.Result
    Mask for Leafs=@n9.G
    DamageMask=@n9.B
    CODE_DamagePercentage=ScalarParameter() {parameterName=Code_CustomDamagePercentage}
   ) {materialFunction=/Game/Packages/Fortnite_UniqueMaterials_1/MaterialFunctions/Trees/MF_Tree_Damage}.Result
   Mask=@n9.G
  ) {materialFunction=/Game/Packages/Fortnite_UniqueMaterials_1/MaterialFunctions/Trees/MF_TreeColorGradient}
  n12 = MaterialFunctionCall!(In=MaterialFunctionCall!(In=MaterialFunctionCall!(In=@n11.Result) {materialFunction=/Game/Packages/Fortnite_UniqueMaterials_1/MaterialFunctions/Trees/MF_Tree_Fresnel}.Result) {materialFunction=/Game/Environments/Props/Winter/Materials/EmissiveTextureWinterTree}.Material Output) {materialFunction=/Game/Packages/Fortnite_UniqueMaterials_1/MaterialFunctions/Trees/MF_Tree_EmissiveTimeOfDay}
  n13 = MaterialFunctionCall!(
   In=MaterialFunctionCall!(In=MaterialFunctionCall!(In=@n12.Result, Height=@n5.G, Mask for Leafs=@n9.G) {materialFunction=/Game/Packages/Fortnite_UniqueMaterials_1/MaterialFunctions/Trees/MF_Tree_PixelDepthOffset}.Result) {materialFunction=/Game/Packages/Fortnite_UniqueMaterials_1/MaterialFunctions/Trees/MF_TreeSplineThickenBranches}.Result
   Mask=@n9.G
  ) {materialFunction=/Game/Packages/Fortnite_UniqueMaterials_1/MaterialFunctions/Trees/MF_Tree_SubSurface}
  n14 = MaterialFunctionCall!(
   ExtraSnowMask
   Simple Mask
   ForceUseInWSNormal
   UseActorPosition
   UseSnowMaskNormal
   UseColorTest
   UseColorTestRock
   In=MaterialFunctionCall!(
    In=MaterialFunctionCall!(In=@n13.Result) {materialFunction=/Game/Packages/Fortnite_UniqueMaterials_1/MaterialFunctions/Trees/MF_Tree_FakeUnderlight}.Result
    FlipTwoSidedSign
    IsTangentSpaceInput
    HasBackFlipFaceNormals
   ) {materialFunction=/Game/Packages/Fortnite_UniqueMaterials_1/MaterialFunctions/Trees/MF_Tree_Snow}.Result
  ) {materialFunction=/Game/Packages/Fortnite_UniqueMaterials_1/MaterialFunctions/Trees/MF_Tree_Snow_Legacy}
  n15 = GetMaterialAttributes(None=MaterialFunctionCall!(
    In=MaterialFunctionCall!(In=@n14.Result) {materialFunction=/Game/Packages/Fortnite_UniqueMaterials_1/MaterialFunctions/Trees/MF_Spec_Roughness_Tweak}.Result
    OpacityMask=@n9.R
    Normal=@n7
   ) {materialFunction=/Game/Packages/Fortnite_UniqueMaterials_1/MaterialFunctions/Trees/MF_TreeParticleShaderOverride}.Result) {attributeGetTypes [#1]}
  n16 = Subtract(A=CameraPositionWS(), B=Add(A=@n15.World Position Offset, B=WorldPosition().XYZ))
  n17 = Add(
   A=@n15.World Position Offset
   B=Multiply(
    A=ScalarParameter() {defaultValue=5, group=PX_Pixel Depth Offset, parameterName=Height Map Pixel Depth offset}
    B=Normalize(VectorInput=@n16)
   )
  ) {desc=WPO}
  n18 = MaterialFunctionCall(WorldPosition=Add(
    A=StaticSwitchParameter(True=@n17, False=@n15.World Position Offset) {group=a_Optimizations, parameterName="Use Pixel Depth Offset (PDO)"}
    B=WorldPosition() {worldPositionShaderOffset=WPT_ExcludeAllShaderOffsets}.XYZ
   )) {materialFunction=&MF_FN_DestructibleDamage}
  n19 = StaticSwitchParameter(True=Add(A=@n18.DamageAnimWPO, B=@n17), False=@n18.DamageAnimWPO) {group=a_Optimizations, parameterName="Use Pixel Depth Offset (PDO)"}
  n20 = MaterialFunctionCall!(In=SetMaterialAttributes!(MaterialAttributes=@n15.MaterialAttributes, World Position Offset=Add(A=@n15.World Position Offset, B=@n19)) {attributeSetTypes [#1]}) {materialFunction=/Game/Packages/Fortnite_UniqueMaterials_1/MaterialFunctions/Trees/MF_CodeReferenceAnimatedVars}
  n21 = StaticSwitch(
   True=MaterialFunctionCall!(Material Attributes=@n20.Result, IsTangentSpaceNormal, In) {materialFunction=/Game/ImpostorBaker/MaterialFunctions/ImposterCaptureSwitch}.Result
   False=@n20.Result
   Value=StaticBool()
  )
  n22 = Subtract(A=WorldPosition().XYZ, B=ObjectPositionWS())
  n23 = Divide(
   A=Add(A=ComponentMask(None=@n22) {b=true}, B=ScalarParameter() {defaultValue=-306.7278, group=Lava, parameterName=EdgeGradientOffset})
   B=ScalarParameter() {defaultValue=54.307, group=Lava, parameterName=EdgeGradientLength}
  )
  n24 = StaticSwitchParameter(
   True=MaterialFunctionCall!(In=@n21, UVs, TransitionGradient=Clamp(None=@n23)) {materialFunction=/Game/Environments/Sets/Vulcano/Materials/MF_LavaEdgeGradientSimple}.Result
   False=@n21
  ) {parameterName=UseLavaEdgeGradient}
  n25 = MaterialFunctionCall(MA In=StaticSwitchParameter(
    True=MaterialFunctionCall!(In=@n24, UseSphericalMappingOverride) {materialFunction=/Game/Packages/Fortress_SharedMaterials/Base_Material_Functions/MF_CurieFallbackEffect}.Result
    False=@n24
   ) {defaultValue=true, group=Curie, parameterName=UseCurie, sortPriority=1}) {materialFunction=&MF_CreativeMoveTool}
  StaticSwitchParameter(
   True=MaterialFunctionCall!(
    IsStaticSnow
    UseAlternateStaticSnowMask
    AlternateStaticSnowMask
    "UseRoad&&OcclusionMask"=StaticBool()
    IsForGrassMeshToPushDown
    AdditionalMaskToRemoveSnow
    IsFallingSnowFX
    UseMinimapCoords
    IncreaseSnowUsingSlope
    UseSmoothBorder=StaticBool() {value=true}
    LerpHighM=Constant() {r=1.5}
    TemporalHeightLerpContrast
    IsLobbyIsland
    UseBR-IslandMask
    IsTree
    IsForRenderTarget
    UseSparkle
    UseSnowNormals
    UseAdditionalMaskToRemoveSnow
    IsTangentSpaceInput
    UseVertexNormalsForBasicSnow
    UseTangentSnowNormals
    Material=@n25.MA Out
   ) {materialFunction=/Game/Athena/Apollo/Environments/Landscape/Materials/Biome/UniqueBiomeMaterialFunctions/MF_Apollo_Snow_Progression}.Material
   False=@n25.MA Out
  ) {group=Snow, parameterName=Use_BRSnow-Progressive}
 ]
}
