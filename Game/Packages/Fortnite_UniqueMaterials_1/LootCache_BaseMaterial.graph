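# Base material for floating loot cache pickups, blending a textured shell with a rarity-colored glowing layer by red vertex color.
# The shell has optional sphere-map reflection and Fresnel; the glow combines world-aligned panning noise with edge masks switchable for sphere, pyramid or cube/cylinder meshes.
# World position offset stacks bobbing, rotation, LOD effect, black-hole pull and optional Creative coin bobble.
# Customized UVs carry rotation-corrected normals to pixel shading.
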Material /Game/Packages/Fortnite_UniqueMaterials_1/LootCache_BaseMaterial
graph {
 properties {
  bUseMaterialAttributes=true
  bUsedWithStaticLighting=true
  numCustomizedUVs=5
  pixelDepthOffsetMode=PDOM_Legacy
  refractionMethod=RM_IndexOfRefraction
  usedShadingModels=Default Lit
 }
 nodes [
  n1 = TextureObjectParameter() {group=Emissive, parameterName=EmissiveTexture, texture=&T_BlacksRGB}
  n2 = VectorParameter() {group=Emissive, parameterName=Emissive Texture UVs RG_TopLeft BA_BottomRight, defaultValue=<rgba 0 0 1 1>}
  n3 = ComponentMask(None=@n2.RGB) {g=true, r=true}
  n4 = AppendVector(A=@n2.B, B=@n2.B)
  n5 = TextureSampleParameter2D() {parameterName=Diffuse, texture=&T_BlacksRGB}
  n6 = Transform(None=ReflectionVectorWS()) {transformSourceType=TRANSFORMSOURCE_World, transformType=TRANSFORM_View}
  n7 = Multiply(A=ConstantBiasScale(None=ComponentMask(None=@n6) {g=true, r=true}), B=Constant2Vector() {g=-1, r=1})
  n8 = TextureSampleParameter2D() {parameterName=ReflectionTextureMask, samplerType=SAMPLERTYPE_Masks, texture=&T_Fortnite_Default_S}
  n9 = LinearInterpolate(
   A=@n5.RGB
   B=Multiply(
    A=TextureSampleParameter2D(UVs=@n7) {parameterName=SphereReflectionMap}.RGB
    B=ScalarParameter() {defaultValue=1, group=reflection, parameterName=Reflection Multiplier}
   )
   Alpha=StaticComponentMaskParameter(None=@n8.RGB) {defaultR=true, group=reflection, parameterName=ReflectionMaskChannel}
  )
  n10 = TextureSampleParameter2D() {
   group=Specular
   parameterName=SpecularMasks
   samplerType=SAMPLERTYPE_Masks
   texture=&T_Fortnite_Default_S
  }
  n11 = TextureCoordinate() {coordinateIndex=3}
  n12 = MaterialFunctionCall(
   BaseNormal=AppendVector(A=TextureCoordinate() {coordinateIndex=2}, B=ComponentMask(None=@n11) {r=true})
   AdditionalNormal=TextureSampleParameter2D() {parameterName=Normals, samplerType=SAMPLERTYPE_Normal, texture=&FlatNormalBlue}.RGB
  ) {materialFunction=&BlendAngleCorrectedNormals}
  n13 = VectorParameter() {group=PickUp_BP Variables, parameterName=WorldPostionOfBlackHole, defaultValue=<rgba -30000 25536 1952 0>}
  n14 = AppendVector(A=@n13.RGB, B=@n13.RGB)
  n15 = Constant3Vector()
  n16 = VectorParameter() {group=WPO, parameterName=Floating Bob Distance, defaultValue=<rgba 0 0 2 0>}
  n17 = ScalarParameter() {defaultValue=0.2, group=WPO, parameterName=Bob Speed, sortPriority=1}
  n18 = Constant3Vector() {constant=<rgba 0 0 1 0>}
  n19 = Multiply(A=Time(), B=ScalarParameter() {defaultValue=0.1, group=WPO, parameterName=RotationSpeed, sortPriority=2})
  n20 = Add(
   A=LinearInterpolate(
    A=Transform(None=@n15) {transformSourceType=TRANSFORMSOURCE_Local}
    B=Transform(None=@n16.RGB) {transformSourceType=TRANSFORMSOURCE_Local}
    Alpha=MaterialFunctionCall(Speed=@n17) {materialFunction=&DebugTimeSine}.Result
   )
   B=RotateAboutAxis(NormalizedRotationAxis=@n18, RotationAngle=@n19, PivotPoint=ObjectPositionWS(), Position=WorldPosition().XYZ)
  )
  n21 = MaterialFunctionCall!(
   World Position=Add(A=WorldPosition().XYZ, B=@n20)
   DistanceCullFadeAnimationOffset=MaterialFunctionCall() {materialFunction=/Game/Packages/Fortress_SharedMaterials/Base_Material_Functions/PickupEffectDistanceCullOverride}.Result
   Use actor center instead of pivot point
  ) {materialFunction=/Game/Packages/Fortress_SharedMaterials/Base_Material_Functions/LODEffect}
  n22 = Add(A=@n21.Result, B=Add(A=WorldPosition().XYZ, B=@n20))
  n23 = Add(
   A=Add(
    A=MaterialFunctionCall!(WorldPostionOfBlackHole=@n14, World Position=@n22) {materialFunction=&/BlackHoleFunction}.Result
    B=@n21.Result
   )
   B=@n20
  )
  n24 = StaticSwitchParameter(
   True=Add(
    A=MaterialFunctionCall!(TimeScale, MaxHeight, MinHeight) {materialFunction=/Game/Playgrounds/Items/Props/Materials/MF_Creative_CoinBobble}.WorldPositionOffset
    B=@n23
   )
   False=@n23
  ) {parameterName=UseCreativeCoinBobble}
  n25 = MaterialFunctionCall(Rotation Axis=@n18, Rotation Angle=@n19) {materialFunction=&FixRotateAboutAxisNormals}
  n26 = MaterialFunctionCall(Float3=@n25.New World Space Vertex Normal) {materialFunction=&BreakOutFloat3Components}
  n27 = AppendVector(A=@n25.New Tangent Space Vertex Normal B, B=@n26.R)
  n28 = AppendVector(A=@n26.G, B=@n26.G)
  n29 = MakeMaterialAttributes!(
   BaseColor=MaterialFunctionCall!(
    DiffuseFresnelAmount=ScalarParameter() {defaultValue=0.5, group=Diffuse, parameterName=DiffuseFresnelAmount}
    Emissive=MaterialFunctionCall!(UVs, TextureIn=@n1, UpperLeftCorner=@n3, LowerRightCorner=@n4, ExportFloat4) {materialFunction=/Engine/Functions/Engine_MaterialFunctions01/Texturing/TextureCropping}.Cropped
    Base Color=StaticSwitchParameter(True=@n9, False=@n5.RGB) {group=reflection, parameterName=Use Reflection}
   ) {materialFunction=/Game/Packages/Fortress_SharedMaterials/Base_Material_Functions/FresnelFunction}.d
   Metallic=Multiply(A=@n10.G, B=ScalarParameter() {defaultValue=1, group=Specular, parameterName=Metallic, sliderMax=1})
   Specular=Multiply(A=@n10.R, B=ScalarParameter() {defaultValue=1, group=Specular, parameterName=Specular Multiply, sliderMax=1})
   Roughness=LinearInterpolate(
    A=ScalarParameter() {defaultValue=0.08, group=Specular, parameterName=SpecRoughnessMin, sliderMax=1}
    B=ScalarParameter() {defaultValue=0.5, group=Specular, parameterName=SpecRoughnessMax, sliderMax=1}
    Alpha=@n10.B
   )
   Anisotropy
   EmissiveColor
   Opacity
   OpacityMask
   Normal=@n12.Result
   Tangent
   WorldPositionOffset=@n24
   SubsurfaceColor
   ClearCoat
   ClearCoatRoughness
   AmbientOcclusion
   Refraction
   CustomizedUVs_0
   CustomizedUVs_1
   CustomizedUVs_2=@n25.New Tangent Space Vertex Normal RG
   CustomizedUVs_3=@n27
   CustomizedUVs_4=@n28
   CustomizedUVs_5
   CustomizedUVs_6
   CustomizedUVs_7
   PixelDepthOffset
   ShadingModel
   Displacement
  )
  n30 = ScalarParameter() {defaultValue=5, group=Diffuse, parameterName=BaseColor_RedVert_Fresnel}
  n31 = ComponentMask(None=VectorParameter() {parameterName=Rarity_Color, defaultValue=<rgba 0.1 1 0.1 1>}.RGB) {b=true, g=true, r=true}
  n32 = TextureCoordinate() {coordinateIndex=3}
  n33 = Normalize(VectorInput=Subtract(A=CameraPositionWS(), B=ObjectPositionWS()))
  n34 = DotProduct(A=AppendVector(A=ComponentMask(None=@n32) {g=true}, B=TextureCoordinate() {coordinateIndex=4}), B=@n33)
  n35 = Constant() {r=8}
  n36 = SphereMask(A=WorldPosition().XYZ, B=ObjectPositionWS(), Radius=ObjectRadius(), Hardness=Constant())
  n37 = Constant() {r=2}
  n38 = Power(
   Base=LinearInterpolate(A=Power(Base=@n34, Exp=@n35) {constExponent=8}, Alpha=Power(Base=@n36, Exp=@n37) {constExponent=0.5}) {constA=1}
   Exp=Constant() {r=2}
  ) {constExponent=10}
  n39 = Clamp(None=DotProduct(A=@n33, B=CameraVectorWS()))
  n40 = PixelDepth()
  n41 = ScalarParameter() {defaultValue=0.01, group=Core Noise, parameterName=N_Min, sortPriority=1}
  n42 = ScalarParameter() {defaultValue=0.13, group=Core Noise, parameterName=N_Max, sortPriority=2}
  n43 = TextureCoordinate() {coordinateIndex=3}
  n44 = MaterialFunctionCall!(
   TextureObject=TextureObject() {texture=/Game/Global/Textures/Default/Noise/T_Tiling_Noise_01}
   TextureSize=VectorParameter() {group=Core Noise, parameterName=TextureSize, sortPriority=3, defaultValue=<rgba 25 25 25 0>}.RGB
   WorldPosition=WorldPosition() {worldPositionShaderOffset=WPT_ExcludeAllShaderOffsets}.XYZ
   Export Float 4
   In=ComponentMask(None=VectorParameter() {group=Core Noise, parameterName=PannerSpeed, sortPriority=4, defaultValue=<rgba -0.1 -0.2 0 0>}.RGB) {g=true, r=true}
   World Space Normal=AppendVector(A=ComponentMask(None=@n43) {g=true}, B=TextureCoordinate() {coordinateIndex=4})
   ProjectionTransitionContrast
  ) {materialFunction=/Game/Athena/SafeZone/WorldAlignedTexture_ScalePan}
  n45 = Desaturation(None=@n44.XYZ Texture)
  n46 = Add(
   A=Multiply(A=Power(Base=@n39, Exp=@n40.R) {constExponent=40}, B=ScalarParameter() {defaultValue=0.1, parameterName=ObjectGlow}) {constB=0.01}
   B=Multiply(A=LinearInterpolate(A=@n41, B=@n42, Alpha=@n45), B=ScalarParameter() {defaultValue=0.8, group=Core Noise, parameterName=NoiseIntensity})
  )
  n47 = MaterialFunctionCall(LDR Input=Multiply(A=@n31, B=Multiply(A=@n38, B=@n46)), HDR Tint=Multiply(A=@n31, B=Constant() {r=6}), Derive HDR Power=Constant() {r=0.5}) {materialFunction=&DeriveHDRfromLDR}
  n48 = ScalarParameter() {defaultValue=1, group=Emissive, parameterName=EmissiveMultiply}
  n49 = SphereMask(
   A=TextureCoordinate() {coordinateIndex=1}
   B=Constant() {r=0.5}
   Radius=ScalarParameter() {defaultValue=0.5, group=Sphere Mesh, parameterName=Radius, sortPriority=1}
   Hardness=ScalarParameter() {group=Sphere Mesh, parameterName=Hardness, sortPriority=2}
  )
  n50 = TextureCoordinate() {coordinateIndex=1}
  n51 = Multiply(
   A=Power(
    Base=TextureSample(UVs=@n50) {samplerType=SAMPLERTYPE_Masks, texture=/Game/Effects/Fort_Effects/Textures/Grad/T_LootCachePyramidGradient}.R
    Exp=ScalarParameter() {defaultValue=1.5, group=Pyramid Mesh, parameterName=PyramidSideContrast, sortPriority=1}
   )
   B=ScalarParameter() {defaultValue=0.25, group=Pyramid Mesh, parameterName=Pyramid Side Intensity, sortPriority=2}
  )
  n52 = TextureSample(UVs=TextureCoordinate() {coordinateIndex=1}) {samplerType=SAMPLERTYPE_LinearGrayscale, texture=/Game/Effects/Fort_Effects/Effects/Characters/Generic/ResIn/SquareGradient_Linear}
  n53 = VertexColor()
  n54 = TextureCoordinate() {coordinateIndex=1}
  n55 = OneMinus(None=Sine(None=ComponentMask(None=@n54) {g=true}) {period=2})
  n56 = ScalarParameter() {defaultValue=1, group=CubeMesh, parameterName=Cube/Cylinder_Edge_Contrast}
  n57 = StaticSwitchParameter(
   True=Power(Base=OneMinus(None=@n49), Exp=ScalarParameter() {defaultValue=65, group=Sphere Mesh, parameterName=Sphere power contrast, sortPriority=3})
   False=StaticSwitchParameter(True=LinearInterpolate(A=@n51, B=@n52.R, Alpha=@n53.G), False=Power(Base=@n55, Exp=@n56)) {group=Edge, parameterName=Is Pyramid Mesh, sortPriority=6}
  ) {group=Edge, parameterName=Is Sphere Mesh, sortPriority=5}
  n58 = Power(
   Base=LinearInterpolate(
    A=@n57
    B=Multiply(A=@n57, B=@n44.XYZ Texture)
    Alpha=ScalarParameter() {defaultValue=0.5, group=Edge, parameterName=AddNoise, sortPriority=4}
   ) {constAlpha=0.45}
   Exp=ScalarParameter() {defaultValue=1.2, group=Edge, parameterName=EdgePower, sortPriority=2}
  ) {constExponent=1.2}
  n59 = Multiply(
   A=Multiply(A=VectorParameter() {group=Edge, parameterName=EdgeColor, defaultValue=<rgba 1 1 1 0>}.RGB, B=Multiply(A=@n58) {constB=10})
   B=ScalarParameter() {defaultValue=1, group=Edge, parameterName=Edge Intensity, sortPriority=1}
  )
  n60 = Multiply(A=ScalarParameter() {defaultValue=0.2, group=Edge, parameterName=Edge_ScaleAddColor, sortPriority=3}, B=Add(A=@n59, B=@n31))
  n61 = MakeMaterialAttributes(
   BaseColor=Multiply(A=Fresnel(ExponentIn=@n30), B=VectorParameter() {group=Diffuse, parameterName=BaseColor_RedVert}.RGB)
   Metallic=Constant()
   Specular=ScalarParameter() {defaultValue=1, group=Specular, parameterName=Metallic_RedVert}
   Roughness=LinearInterpolate(Alpha=TextureSample() {texture=/Game/Effects/Fort_Effects/Textures/Misc/T_Shot0030_NoiseCutout}.G) {constB=0.1}
   EmissiveColor=Add(A=Multiply(A=@n59, B=@n60), B=Multiply(A=@n47.RemappedRange, B=@n48))
   Normal=@n12.Result
   WorldPositionOffset=@n24
   CustomizedUVs_2=@n25.New Tangent Space Vertex Normal RG
   CustomizedUVs_3=@n27
   CustomizedUVs_4=@n28
  )
  n62 = VertexColor()
  MaterialFunctionCall(MA In=MaterialFunctionCall!(Material attributes=BlendMaterialAttributes(A=@n29, B=@n61, Alpha=@n62.R), IsGameplayAsset=StaticBool()) {materialFunction=/Game/Packages/Fortress_SharedMaterials/Base_Material_Functions/Fortnite_BaseFunction}.Material Attributes) {materialFunction=&MF_CreativeMoveTool}
 ]
}
