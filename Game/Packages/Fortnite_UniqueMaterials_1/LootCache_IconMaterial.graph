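# Translucent unlit material for a loot cache icon, showing a configurable icon over an animated backdrop, tinted by rarity color.
# Emissive turns a radial rarity-color gradient into HDR and pulses it on a sine.
# Opacity adds the icon to a scorch-masked gradient distorted by a radially panned normal map, and is zero unless Use Icon is on.
# World position offset combines bobbing, Z-axis rotation, distance LOD and black-hole pickup functions.
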Material /Game/Packages/Fortnite_UniqueMaterials_1/LootCache_IconMaterial
graph {
 properties {
  bUseMaterialAttributes=true
  bUsedWithStaticLighting=true
  blendMode=BLEND_Translucent
  pixelDepthOffsetMode=PDOM_Legacy
  refractionMethod=RM_IndexOfRefraction
  shadingModel=MSM_Unlit
  usedShadingModels=Unlit
 }
 nodes [
  n1 = Constant() {r=0.5}
  n2 = SphereMask(A=TextureCoordinate(), B=@n1, Radius=@n1) {hardnessPercent=0}
  n3 = Multiply(
   A=ComponentMask(None=VectorParameter() {parameterName=Rarity_Color, defaultValue=<rgba 0.1 1 0.1 1>}.RGB) {b=true, g=true, r=true}
   B=LinearInterpolate(Alpha=OneMinus(None=@n2)) {constA=1, constB=0.25}
  )
  n4 = MaterialFunctionCall(LDR Input=@n3, HDR Tint=Multiply(A=@n3, B=Constant() {r=6}), Derive HDR Power=Constant() {r=0.5}) {materialFunction=&DeriveHDRfromLDR}
  n5 = ScalarParameter() {defaultValue=1, parameterName=EmissiveMultiply}
  n6 = LinearInterpolate(
   A=@n5
   B=Add(A=@n5, B=Multiply(A=@n5) {constB=0.25})
   Alpha=MaterialFunctionCall(Speed=ScalarParameter() {defaultValue=0.2, group=WPO, parameterName=Bob Speed, sortPriority=1}) {materialFunction=&DebugTimeSine}.Result
  ) {constA=1, constB=1.25}
  n7 = TextureCoordinate()
  n8 = Panner(Coordinate=MaterialFunctionCall() {materialFunction=&VectorToRadialValue}.Radial Coordinates) {speedY=0.3}
  n9 = ComponentMask(None=MaterialFunctionCall(
    Normal=TextureSample(UVs=@n8) {samplerType=SAMPLERTYPE_Normal, texture=/Game/Effects/Fort_Effects/Textures/Husk/Poison/T_PoisonTile_N}.RGB
    Flatness=ScalarParameter() {defaultValue=0.25}
   ) {materialFunction=&FlattenNormal}.Result) {g=true, r=true}
  n10 = TextureSample(UVs=BumpOffset(Coordinate=Add(A=@n7, B=@n9), Height=ScalarParameter() {defaultValue=-0.5, parameterName=Bump_Height})) {texture=/Game/Effects/Fort_Effects/Effects/Characters/Generic/ResIn/SquareGradient}
  n11 = TextureSample() {texture=/Game/Effects/Fort_Effects/Gameplay/SmokeDustExplosions/Textures/T_Decal_Scorch_Hole}
  n12 = ComponentMask(None=VectorParameter() {parameterName=Icon_Scale, defaultValue=<rgba 1 1 1 1>}.RGB) {g=true, r=true}
  n13 = VectorParameter() {parameterName=Icon_Placement}
  n14 = TextureSampleParameter2D(UVs=Add(
    A=MaterialFunctionCall(Texture Scale=@n12) {materialFunction=&ScaleUVsByCenter}.UVs
    B=ComponentMask(None=@n13.RGB) {g=true, r=true}
   )) {parameterName=Icon Texture, texture=/Game/UI/Foundation/Textures/Icons/Quest/T-Icon-DurrBurg-128}
  n15 = Add(
   A=Multiply(A=Multiply(A=@n10.R, B=@n11.R), B=ScalarParameter() {defaultValue=1, parameterName=BKGD_Intensity})
   B=Desaturation(None=Multiply(A=@n14.RGB, B=@n14.RGB))
  )
  n16 = Constant()
  n17 = VectorParameter() {group=PickUp_BP Variables, parameterName=WorldPostionOfBlackHole, defaultValue=<rgba -30000 25536 1952 0>}
  n18 = Constant3Vector()
  n19 = VectorParameter() {group=WPO, parameterName=Floating Bob Distance, defaultValue=<rgba 0 0 2 0>}
  n20 = ScalarParameter() {defaultValue=0.2, group=WPO, parameterName=Bob Speed, sortPriority=1}
  n21 = Time()
  n22 = ScalarParameter() {defaultValue=0.1, parameterName=RotationSpeed, sortPriority=2}
  n23 = Add(
   A=LinearInterpolate(
    A=Transform(None=@n18) {transformSourceType=TRANSFORMSOURCE_Local}
    B=Transform(None=@n19.RGB) {transformSourceType=TRANSFORMSOURCE_Local}
    Alpha=MaterialFunctionCall(Speed=@n20) {materialFunction=&DebugTimeSine}.Result
   )
   B=RotateAboutAxis(
    NormalizedRotationAxis=Constant3Vector() {constant=<rgba 0 0 1 0>}
    RotationAngle=Multiply(A=@n21, B=@n22)
    PivotPoint=ObjectPositionWS()
    Position=WorldPosition().XYZ
   )
  )
  n24 = MaterialFunctionCall!(
   World Position=Add(A=WorldPosition().XYZ, B=@n23)
   DistanceCullFadeAnimationOffset=MaterialFunctionCall() {materialFunction=/Game/Packages/Fortress_SharedMaterials/Base_Material_Functions/PickupEffectDistanceCullOverride}.Result
   Use actor center instead of pivot point
  ) {materialFunction=/Game/Packages/Fortress_SharedMaterials/Base_Material_Functions/LODEffect}
  n25 = Add(A=WorldPosition().XYZ, B=@n23)
  n26 = Add(
   A=MaterialFunctionCall!(WorldPostionOfBlackHole=AppendVector(A=@n17.RGB, B=@n17.RGB), World Position=Add(A=@n24.Result, B=@n25)) {materialFunction=&/BlackHoleFunction}.Result
   B=@n24.Result
  )
  MaterialFunctionCall!(
   Material attributes=MakeMaterialAttributes!(
    BaseColor
    Metallic
    Specular
    Roughness
    Anisotropy
    EmissiveColor=Multiply(A=@n4.RemappedRange, B=@n6)
    Opacity=StaticSwitchParameter(True=@n15, False=@n16) {parameterName=Use Icon}
    OpacityMask
    Normal
    Tangent
    WorldPositionOffset=Add(A=@n26, B=@n23)
    SubsurfaceColor
    ClearCoat
    ClearCoatRoughness
    AmbientOcclusion
    Refraction
    CustomizedUVs_0
    CustomizedUVs_1
    CustomizedUVs_2
    CustomizedUVs_3
    CustomizedUVs_4
    CustomizedUVs_5
    CustomizedUVs_6
    CustomizedUVs_7
    PixelDepthOffset
    ShadingModel
    Displacement
   )
   IsGameplayAsset=StaticBool()
  ) {materialFunction=/Game/Packages/Fortress_SharedMaterials/Base_Material_Functions/Fortnite_BaseFunction}
 ]
}
