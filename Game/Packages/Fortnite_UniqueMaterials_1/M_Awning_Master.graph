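# Masked, two-sided master material for striped awnings, with a packed diffuse texture and two tintable stripe colors.
# Diffuse R blends the squared stripe colors, G multiplies the result into base color, and B drives the opacity mask.
# The deprecated base shader function combines that color with specular masks, then the attributes pass through MF_UpVectorTexture and MF_CreativeMoveTool.
# Destruction, hit-animation and DiffuseColoration parameters have no visible connection to the output.
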Material /Game/Packages/Fortnite_UniqueMaterials_1/M_Awning_Master
graph {
 properties {
  bUseMaterialAttributes=true
  bUsedWithInstancedStaticMeshes=true
  bUsedWithNanite=true
  bUsedWithStaticLighting=true
  blendMode=BLEND_Masked
  pixelDepthOffsetMode=PDOM_Legacy
  twoSided=true
  usedShadingModels=Default Lit
 }
 nodes [
  n1 = VectorParameter() {group=Code, parameterName=DiffuseColoration, defaultValue=<rgba 1 1 1 0>}
  AppendVector(A=@n1.RGB, B=@n1.RGB)
  n2 = TextureSampleParameter2D() {parameterName=Diffuse, texture=/Game/Packages/Fortress_ExteriorDecos01/SM/Materials/T_FORT_Awning}
  n3 = VectorParameter() {parameterName=StripeColor_A, defaultValue=<rgba 0.1399 0.3253 0.6011 1>}
  n4 = VectorParameter() {parameterName=StripeColor_B, defaultValue=<rgba 1 1 1 4>}
  n5 = LinearInterpolate(A=Multiply(A=@n3.RGB, B=@n3.RGB), B=Multiply(A=@n4.RGB, B=@n4.RGB), Alpha=@n2.R)
  n6 = TextureSampleParameter2D() {
   group=Specular
   parameterName=SpecularMasks
   samplerType=SAMPLERTYPE_Masks
   texture=/Game/Packages/Fortress_ExteriorDecos01/SM/Materials/T_FORT_Awnings_S
  }
  n7 = MaterialFunctionCall!(
   -----------Base Shader-----------
   BaseColor1=Clamp(None=Multiply(A=@n2.G, B=@n5))
   Base Color Alpha
   SpecularMasks1=@n6.RGB
   Emissive
   Normals1
   World Position Offset
   DiffuseFresnelAmount
   -----------Destruction------------
   ------Multiple Diffuse Textures-----
   BaseColor2
   BaseColor3
   BaseColor4
   BaseColor5
   BaseColor6
   BaseColor7
   BaseColor8
   " - - - - - - - -"
   SpecularMasks2
   SpecularMasks3
   SpecularMasks4
   SpecularMasks5
   SpecularMasks6
   SpecularMasks7
   SpecularMasks8
   " - - - - - - - - "
   normal2
   Normal3
   Normal4
   Normal5
   Normal6
   Normal7
   Normal8
   " - - - - - - - - 1"
   Use 2 Textures
   Use 3 Textures
   Use 4 Textures
   Use 5 Textures
   Use 6 Textures
   Use 7 Textures
   Use 8 Textures
   Use 9 Textures
   Use 10 Textures
   ExposeLerpFunctionForInvertingTheHitDistanceCalc
  ) {materialFunction=/Game/Packages/Fortress_SharedMaterials/Base_Material_Functions/DEPRECATED_BaseShaderFunction}
  n8 = SetMaterialAttributes!(
   MaterialAttributes
   Base Color=@n7.Diffuse
   Metallic=@n7.Diffuse
   Specular=@n7.Diffuse
   Roughness=@n7.Diffuse
   Emissive Color=@n7.Diffuse
   Opacity Mask=@n2.B
   Normal=@n7.Diffuse
   World Position Offset=@n7.Diffuse
  ) {attributeSetTypes [
    #1
    #2
    #3
    #4
    #5
    #6
    #7
    #8
   ]}
  MaterialFunctionCall(MA In=MaterialFunctionCall!(In=@n8, UpVectorTexture) {materialFunction=/Game/Athena/WinterArt/Materials/MF_UpVectorTexture}.Result) {materialFunction=&MF_CreativeMoveTool}
  ScalarParameter() {group=Code, parameterName=Code_HitAnimation}
  ScalarParameter() {group=WorldPositionOffset, parameterName=DistanceCullFadeAnimationOffset}
  ScalarParameter() {group=Code, parameterName=Code_DestructionAmount}
  ScalarParameter() {group=Code, parameterName=WSDamageRadius}
  VectorParameter() {parameterName=WSDeformationCenter}
  VectorParameter() {parameterName=Code_DeformationVector}
 ]
}
