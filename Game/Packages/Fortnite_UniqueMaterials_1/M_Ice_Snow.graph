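# Subsurface ice material, usable on mesh particles, with animated cracking and a timed wobble.
# A Crack particle parameter offsets a floored noise mask that brings in crack normals and darkening over a rock normal.
# Emissive terms combine tinted skylight cubemap reflection, a sun-direction specular highlight and Fresnel edge glow.
# A sine wobble along vertex normals is blended by an eased timer from the WobbleTime particle parameter and FlashDuration.
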Material /Game/Packages/Fortnite_UniqueMaterials_1/M_Ice_Snow
graph {
 properties {
  bUsedWithMeshParticles=true
  pixelDepthOffsetMode=PDOM_Legacy
  refractionMethod=RM_IndexOfRefraction
  shadingModel=MSM_Subsurface
  usedShadingModels=Subsurface
 }
 nodes [
  Add() {constB=0.1}
  n1 = TextureCoordinate() {uTiling=2, vTiling=2}
  n2 = TextureSample(UVs=@n1) {texture=/Game/Effects/Fort_Effects/Textures/Tile/T_Glass_Cracks_M}
  n3 = Multiply(A=@n2.R) {constB=0.5}
  n4 = Multiply(
   A=TextureSample(UVs=@n1) {samplerType=SAMPLERTYPE_Normal, texture=/Game/Effects/Fort_Effects/Textures/Tile/T_Glass_Cracks_N}.RGB
   B=Constant3Vector() {constant=<rgba 3 3 0 0>}
  )
  n5 = DynamicParameter() {defaultValue=<rgba 0 1 1 1>, paramNames [Crack,WobbleTime,Param3,Param4]}
  n6 = Floor(None=Add(A=TextureSample(UVs=@n1) {texture=&T_Smooth_tile}.G, B=@n5.Crack))
  Add(
   A=TextureSampleParameter2D() {parameterName=Base Normal, samplerType=SAMPLERTYPE_Normal, texture=/Game/Environments/Props/Rocks/Textures/T_Corrupt_Rock_AB_N}.RGB
   B=Multiply(A=Add(A=@n3, B=@n4), B=@n6)
  )
  n7 = TextureCoordinate() {uTiling=0.3, vTiling=0.3}
  Add(
   A=Fresnel() {baseReflectFraction=0.2, exponent=0.2}
   B=OneMinus(None=TextureSample(UVs=@n7) {samplerType=SAMPLERTYPE_LinearGrayscale, texture=/Game/Global/Textures/Default/Tileable_Patterns/T_StormCloudDetail01_M}.RGB)
  )
  Add()
  Add()
  Constant() {r=0.3}
  n8 = Add(A=Sine(None=Time()) {period=0.09}) {constB=0.2}
  n9 = ScalarParameter() {defaultValue=2.2, parameterName=WobbleAmount}
  n10 = Subtract(A=Time() {period=5}, B=@n5.WobbleTime)
  n11 = ScalarParameter() {defaultValue=0.64, group=Timestamp, parameterName=FlashDuration}
  n12 = MaterialFunctionCall!(In=Clamp(None=Divide(A=@n10, B=@n11))) {materialFunction=/Game/Effects/Fort_Effects/Materials/_Functions/MF_Smoothstep_EaseOut}
  LinearInterpolate(B=Multiply(A=VertexNormalWS(), B=Multiply(A=@n8, B=@n9)), Alpha=Clamp(None=Multiply(A=@n12.Result) {constB=1.5})) {constAlpha=1}
  MaterialFunctionCall() {materialFunction=&DebugTimeSine}
  n13 = TextureSampleParameterCube!(UVs=ReflectionVectorWS() {bNormalizeCustomWorldNormal=true}, Apply View MipBias) {
   parameterName=MaterialExpressionTextureSampleParameterCube_0
   samplerType=SAMPLERTYPE_LinearColor
   texture=/Game/Packages/Fortress_Sky/TexturesHDR/T_AthenaSkylight
  }
  n14 = Multiply(
   A=Multiply(A=@n13.RGB, B=VectorParameter() {parameterName=Cube Map Multiplier, defaultValue=<rgba 0.3282 0.5416 0.7 0>}.RGB) {constB=1.5}
   B=ComponentMask(None=MaterialFunctionCall() {materialFunction=/Engine/Functions/GetPostProcessFunctions/GetAmbientCubemapTint}.AmbientCubemapTint) {b=true, g=true, r=true}
  )
  n15 = Constant() {r=0.5}
  n16 = CollectionParameter() {
   collection=&FortniteMaterialParameters
   parameterName=SunAndMoonModelDirectionalVector
  }
  n17 = Clamp(None=DotProduct(A=ComponentMask(None=@n16) {b=true, g=true, r=true}, B=ReflectionVectorWS()))
  n18 = ScalarParameter() {defaultValue=70, parameterName=Spec Highlight Power}
  n19 = Add(
   A=ShadingPathSwitch(Default=Desaturation(None=@n14, Fraction=@n15), Mobile=Constant()) {desc=Migrated from FeatureLevelSwitch}
   B=Multiply(A=Power(Base=@n17, Exp=@n18) {constExponent=20}, B=ScalarParameter() {defaultValue=3, parameterName=Spec Intense})
  )
  n20 = Clamp(None=LinearInterpolate() {constA=0.72, constB=-3})
  n21 = VectorParameter() {parameterName=EdgeGlow, defaultValue=<rgba 0.3917 0.6351 1.6 0>}
  n22 = Fresnel() {baseReflectFraction=0, exponent=2}
  n23 = Multiply(B=Add(A=VectorParameter() {parameterName=BaseColor, defaultValue=<rgba 0.0065 0.0279 0.1771 0>}.RGB, B=Multiply(A=@n21.RGB, B=@n22)))
  n24 = LinearInterpolate(B=OneMinus(None=@n2.G), Alpha=@n6) {constA=1}
  MaterialFunctionCall(Emissive=Add(A=Multiply(A=@n19, B=@n20), B=Multiply(A=@n23, B=@n24))) {materialFunction=&MF_FX_EmissiveAdjustments}
  Multiply(A=VectorParameter() {parameterName=Subsurface Color, defaultValue=<rgba 0.4415 0.7305 0.9115 0>}.RGB, B=@n22)
  ScalarParameter() {defaultValue=2, parameterName=Refraction}
  TextureSample() {texture=/Game/Athena/Items/Traps/FreezeTrap/FX/T_IceCube_Cheap_D}
 ]
}
