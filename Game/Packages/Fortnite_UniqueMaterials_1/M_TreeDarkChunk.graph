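# Translucent particle sprite material for a dark tree chunk effect, usable with both Niagara and Cascade sprites.
# It multiplies particle color by the green channel of the tree chunk mask texture and feeds that into the MF_FX_EmissiveAdjustments function.
# A second multiply combines particle alpha with the same mask channel.
# Shading is Default Lit with volumetric per-vertex non-directional translucency lighting.
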Material /Game/Packages/Fortnite_UniqueMaterials_1/M_TreeDarkChunk
graph {
 properties {
  bUsedWithNiagaraSprites=true
  bUsedWithParticleSprites=true
  blendMode=BLEND_Translucent
  pixelDepthOffsetMode=PDOM_Legacy
  translucencyLightingMode=TLM_VolumetricPerVertexNonDirectional
  usedShadingModels=Default Lit
 }
 nodes [
  n1 = ParticleColor()
  n2 = TextureSample() {texture=/Game/Environments/Props/Trees/Textures/T_Tree_DarkChunk_Mask}
  MaterialFunctionCall(Emissive=Multiply(A=@n1.RGB, B=@n2.G)) {materialFunction=&MF_FX_EmissiveAdjustments}
  Multiply(A=@n1.A, B=@n2.G)
 ]
}
