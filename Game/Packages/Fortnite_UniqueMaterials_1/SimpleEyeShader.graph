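# Masked subsurface eye material for skeletal characters, with tintable iris, sclera and eye shadow.
# A scaled or cropped UV feeds a sphere-mask bump offset that gives the iris parallax depth.
# A concave normal drives view-space reflection lookups for inner-eye highlights and fake specular, combined with a channel-packed glint, streak and mask texture.
# An optional gradient multiplies in static shadowing, and a character res-in function is included.
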Material /Game/Packages/Fortnite_UniqueMaterials_1/SimpleEyeShader
graph {
 properties {
  bUsedWithMeshDeformer=true
  bUsedWithMorphTargets=true
  bUsedWithSkeletalMesh=true
  blendMode=BLEND_Masked
  pixelDepthOffsetMode=PDOM_Legacy
  shadingModel=MSM_Subsurface
  usedShadingModels=Subsurface
 }
 nodes [
  Constant()
  n1 = ScalarParameter() {defaultValue=4, group=Color, parameterName=UV Scale}
  n2 = Multiply(A=TextureCoordinate(), B=@n1)
  n3 = Multiply(A=@n1) {constB=0.5}
  n4 = VectorParameter() {group=UV_Coordinates, parameterName=TexturePosition_RG_TopLeft BA_BottomRight, defaultValue=<rgba 0.35 0.35 0.65 0.65>}
  n5 = StaticSwitchParameter(
   True=Add(A=Subtract(A=@n2, B=@n3) {constB=0}) {constB=0.5}
   False=MaterialFunctionCall!(UVs, TextureIn, UpperLeftCorner=ComponentMask(None=@n4.RGB) {g=true, r=true}, LowerRightCorner=AppendVector(A=@n4.B, B=@n4.B), ExportFloat4) {materialFunction=/Engine/Functions/Engine_MaterialFunctions01/Texturing/TextureCropping}.Crop UVs
  ) {defaultValue=true, group=Color, parameterName=UseSimpleUVScale}
  n6 = TextureSample(UVs=@n5) {
   desc=r glint g black and white streaks - b black mask - shadowed eye
   samplerType=SAMPLERTYPE_LinearColor
   texture=/Game/Packages/Fortnite_UniqueMaterials_1/Textures/New_eye_Diffuse2
  }
  LinearInterpolate(
   A=ScalarParameter() {defaultValue=0.2, group=Spec, parameterName=IrisRoughness}
   B=ScalarParameter() {defaultValue=0.25, group=Spec, parameterName=WhiteRoughness}
   Alpha=@n6.B
  )
  MaterialFunctionCall!(Animation Gradient, Rez In Percentage) {materialFunction=/Game/Packages/Fortress_SharedMaterials/Base_Material_Functions/CharacterResInEffect}
  ScalarParameter() {defaultValue=0.9, group=SSS, parameterName=SSS}
  n7 = Constant() {r=0.5}
  n8 = MaterialFunctionCall!(
   HeightTexture=OneMinus(None=SphereMask(A=@n5, B=@n7) {attenuationRadius=0.3, hardnessPercent=0})
   TextureCoordinates=@n5
   ReferencePlane=Constant() {r=1}
   HeightRatioInput=Constant() {r=0.25}
   CameraVector
  ) {materialFunction=/Engine/Functions/Engine_MaterialFunctions02/Utility/BumpOffset_Advanced.BumpOffset_advanced}
  n9 = Normalize(VectorInput=Multiply(
    A=TextureSample(UVs=@n8.UVs) {samplerType=SAMPLERTYPE_Normal, texture=/Game/Packages/Fortnite_UniqueMaterials_1/Textures/ConcaveEyeTest}.RGB
    B=Constant3Vector() {constant=<rgba -0.5 -0.5 1 0>}
   ) {constB=-1})
  n10 = TextureSample(UVs=@n8.UVs) {
   desc=r glint g black and white streaks - b black mask - shadowed eye
   samplerType=SAMPLERTYPE_LinearColor
   texture=&New_eye_Diffuse2
  }
  n11 = Add(
   A=MaterialFunctionCall!(Normal=Transform(None=@n9), CameraVector) {materialFunction=/Engine/Functions/Engine_MaterialFunctions02/Utility/CustomReflectionVector}.Reflection Vector
   B=Multiply(A=CameraVectorWS(), B=@n10.A)
  )
  n12 = ConstantBiasScale(None=ComponentMask(None=Transform(None=@n11) {transformSourceType=TRANSFORMSOURCE_World, transformType=TRANSFORM_View}) {g=true, r=true}) {bias=-0.5, scale=0.25}
  n13 = TextureSample!(UVs=Add(A=@n12, B=@n8.Distortion), Tex, Level=Constant() {r=3}, Apply View MipBias) {desc=inner eye, mipValueMode=TMVM_MipLevel, texture=/Game/Packages/Fortress_SharedMaterials/Textures/ReflectionMap2}
  n14 = Multiply(
   A=LinearInterpolate(
    A=VectorParameter() {group=Color, parameterName=Iris Color 1, defaultValue=<rgba 0.395 0.6231 0.9381 1>}.RGB
    B=VectorParameter() {group=Color, parameterName=Iris Color 2, defaultValue=<rgba 0.1132 0.1865 0.2389 1>}.RGB
    Alpha=@n10.A
   )
   B=@n10.G
  )
  n15 = Power(Base=Multiply(A=Multiply(A=@n14) {constB=2}, B=@n10.A), Exp=Add(A=OneMinus(None=@n10.A)))
  n16 = Multiply(A=Multiply(A=Desaturation(None=@n13.RGB)) {constB=5}, B=@n15)
  n17 = Add(
   A=LinearInterpolate(
    A=Add(A=@n16, B=@n15)
    B=VectorParameter() {group=Color, parameterName=Sclera Color, defaultValue=<rgba 0.6195 0.5133 0.4527 1>}.RGB
    Alpha=@n10.B
   )
   B=@n6.R
  )
  n18 = Add(
   A=Multiply(
    A=TextureSample!(UVs=@n12, Tex, Apply View MipBias) {desc=fake spec, texture=/Game/Items/Art_noLOD/Textures/reflection}.RGB
    B=LinearInterpolate(Alpha=@n10.B) {constA=0.2, constB=0.2}
   )
   B=@n17
  )
  n19 = ScalarParameter() {defaultValue=2.5, group=Shadow, parameterName=UV Scale eye shadow}
  n20 = Multiply(A=TextureCoordinate(), B=@n19)
  n21 = Multiply(A=@n19) {constB=0.5}
  n22 = TextureSample(UVs=Add(A=Subtract(A=@n20, B=@n21) {constB=0}) {constB=0.5}) {samplerType=SAMPLERTYPE_LinearColor, texture=/Game/Packages/Fortnite_UniqueMaterials_1/Textures/Eye-Gradient}
  n23 = Multiply(
   A=LinearInterpolate(
    A=VectorParameter() {group=Shadow, parameterName=EyeShadowColor, defaultValue=<rgba 0.1504 0.0902 0.0887 1>}.RGB
    B=VectorParameter() {group=Shadow, parameterName=LightEyeShadowColor, defaultValue=<rgba 1 1 1 1>}.RGB
    Alpha=@n22.R
   )
   B=@n22.G
  )
  StaticSwitchParameter(True=Multiply(A=@n18, B=@n23), False=@n18) {defaultValue=true, group=Shadow, parameterName=include staticshadowing}
  n24 = Desaturation(None=@n17, Fraction=ScalarParameter() {defaultValue=-2, group=SSS, parameterName=SSS saturation})
  StaticSwitchParameter(True=Multiply(A=@n23, B=@n24), False=@n24) {defaultValue=true, group=Shadow, parameterName=include staticshadowing}
  n25 = Add(A=@n10.A) {constB=0}
  StaticSwitchParameter(True=Multiply(A=@n23, B=@n25), False=@n25) {defaultValue=true, group=Shadow, parameterName=include staticshadowing}
 ]
}
