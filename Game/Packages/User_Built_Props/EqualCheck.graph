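# EqualCheck is a debug material function, exposed to the library under Misc and Debug, that visualizes a comparison of two scalars as a color.
# An If node compares input a against input b and outputs red when a is greater, green when equal, and blue when less.
# Both inputs can fall back to their preview values when left unconnected.
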MaterialFunction /Game/Packages/User_Built_Props/EqualCheck
graph {
 properties {bExposeToLibrary=true, description="Red if A > B  Green if A== B and Blue if A < B", libraryCategoriesText [Misc,Debug]}
 nodes [
  n1 = FunctionInput() {bUsePreviewValueAsDefault=true, inputName=b, inputType=FunctionInput_Scalar, sortPriority=1, previewValue=<1 0 0 1>}
  FunctionOutput(None=If(
    A=FunctionInput() {bUsePreviewValueAsDefault=true, inputName=a, inputType=FunctionInput_Scalar}
    B=@n1
    "A > B"=Constant3Vector() {constant=<rgba 1 0 0 0>}
    "A == B"=Constant3Vector() {constant=<rgba 0 1 0 0>}
    "A < B"=Constant3Vector() {constant=<rgba 0 0 1 0>}
   ))
 ]
}
