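# Post-process material, applied before DOF, that draws pulsing highlights and outlines on stencil-tagged objects, with branches for skin emissive, sharpening and a Creative-mode outline bypass.
# Resolution-scaled four-neighbor custom depth samples form edges, and a depth comparison masks occlusion, inverted for the player outline ID.
# Selection and distance-faded gameplay stencil IDs choose pulsed scene fill and time-of-day-scaled edge colors.
# The skin branch masks emissive by shading model, opacity and inverse metallic.
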Material /Game/Packages/Fortnite_UniqueMaterials_1/SkinAndMeshOutlinePostProcess
graph {
 properties {
  bEnableResponsiveAA=true
  blendMode=BLEND_Translucent
  blendableLocation=BL_SceneColorBeforeDOF
  blendablePriority=-5
  materialDomain=MD_PostProcess
  pixelDepthOffsetMode=PDOM_Legacy
  shadingModel=MSM_Unlit
  usedShadingModels=Unlit
 }
 nodes [
  n1 = ComponentMask(None=SceneTexture() {sceneTextureId=PPI_PostProcessInput0}.Color) {b=true, g=true, r=true}
  n2 = VectorParameter() {parameterName=SceneModulation1, defaultValue=<rgba 1 1 1 1>}
  n3 = VectorParameter() {parameterName=SceneModulation2, defaultValue=<rgba 2 2 2 1>}
  n4 = Time()
  n5 = ScalarParameter() {defaultValue=0.5, parameterName=PulseRate}
  n6 = ConstantBiasScale(None=Sine(None=Multiply(A=@n4, B=@n5)))
  n7 = CollectionParameter() {
   collection=&FortniteMaterialParameters
   parameterName=FortniteActiveTimeOfDay
  }
  n8 = Constant4Vector() {constant=<rgba 1 1 1 0.4>}
  n9 = VectorParameter() {desc=code controled, group=Outline, parameterName=OutlineColor, defaultValue=<rgba 1 1 1 0>}
  n10 = VectorParameter() {desc=code controled, group=Outline, parameterName=OutlineColor2, defaultValue=<rgba 0.059 0.7552 0.5574 0>}
  n11 = MaterialFunctionCall!(ID=Constant() {r=19}) {materialFunction=/Game/Packages/Fortress_SharedMaterials/Stencil_ID_Definitions/StencilMaskCompare}
  n12 = ComponentMask(None=ScreenPosition().ViewportUV) {g=true, r=true}
  n13 = Subtract(A=Divide(A=ViewSize(), B=Constant() {r=320}), B=Constant() {r=2}) {constB=0}
  n14 = SceneTexelSize()
  n15 = Multiply(A=Clamp(None=Ceil(None=@n13)) {maxDefault=2, minDefault=1}, B=@n14) {constA=2}
  n16 = Constant()
  n17 = AppendVector(A=ComponentMask(None=@n15) {desc=x, r=true}, B=@n16)
  n18 = Add(A=@n12, B=@n17) {desc=x+}
  n19 = AppendVector(A=@n16, B=ComponentMask(None=@n15) {desc=y, g=true}) {desc=y+}
  n20 = Add(A=@n12, B=@n19) {desc=y+}
  n21 = Constant() {r=-1}
  n22 = Add(A=@n12, B=Multiply(A=@n17, B=@n21)) {desc=x-}
  n23 = Add(A=@n12, B=Multiply(A=@n21, B=@n19)) {desc=y-}
  n24 = ComponentMask(None=MaterialFunctionCall!(
    Base=SceneTexture() {sceneTextureId=PPI_CustomDepth}.Color
    BaseSampleBias
    Scale
    Sample1=SceneTexture(UVs=@n18) {sceneTextureId=PPI_CustomDepth}.Color
    Sample2=SceneTexture(UVs=@n20) {sceneTextureId=PPI_CustomDepth}.Color
    Sample3=SceneTexture(UVs=@n22) {sceneTextureId=PPI_CustomDepth}.Color
    Sample4=SceneTexture(UVs=@n23) {sceneTextureId=PPI_CustomDepth}.Color
    Divisor=Constant() {r=-16}
    UseCenteredLines
    UseRGB
   ) {materialFunction=/Game/Packages/jl_tests/combine4Samples}.Result) {r=true}
  n25 = LinearInterpolate(
   A=Multiply(A=@n1, B=LinearInterpolate(A=@n2.RGB, B=@n3.RGB, Alpha=@n6))
   B=Multiply(A=DotProduct(A=@n7, B=@n8.RGBA), B=LinearInterpolate(A=@n9.RGB, B=@n10.RGB, Alpha=@n11.Result))
   Alpha=@n24
  ) {desc=outline z tested objects when opaque}
  n26 = Add(A=SceneTexture() {sceneTextureId=PPI_CustomDepth}.Color, B=ScalarParameter() {defaultValue=-2, parameterName=Custom Depth Offset Amount})
  n27 = SceneTexture() {sceneTextureId=PPI_SceneDepth}
  n28 = CollectionParameter() {
   collection=/Game/Packages/Fortress_SharedMaterials/Stencil_ID_Definitions/Stencil_ID_Definitions
   parameterName=PlayerOutlineID_CodeReferenced
  }
  n29 = MaterialFunctionCall(ID=@n28) {materialFunction=&StencilMaskCompare}
  n30 = Multiply(A=ComponentMask(None=Subtract(A=@n26, B=@n27.Color)) {r=true}, B=Add(A=Multiply(A=@n29.Result) {constB=-2})) {desc="If player outline, invert occlusion check"}
  n31 = OneMinus(None=Ceil(None=Clamp(None=@n30)) {desc=z tested objects black z tested - everything else white}) {desc=occlusion mask}
  n32 = CollectionParameter() {
   collection=&Stencil_ID_Definitions
   parameterName=SelectionOutlineID_CodeReferenced
  }
  n33 = CollectionParameter() {
   collection=&Stencil_ID_Definitions
   parameterName=SelectionOutlineID2_CodeReferenced
  }
  n34 = MaterialFunctionCall(ID=@n33) {materialFunction=&StencilMaskCompare}
  n35 = Multiply(
   A=@n31
   B=Max(
    A=MaterialFunctionCall(ID=@n32) {materialFunction=&StencilMaskCompare}.Result
    B=@n34.Result
   )
  )
  n36 = Multiply(
   A=@n1
   B=LinearInterpolate(B=VectorParameter() {parameterName=Quest Fill Color, defaultValue=<rgba 1 0.7929 0.535 1>}.RGB, Alpha=@n6) {constA=1}
  )
  n37 = CollectionParameter() {
   collection=&FortniteMaterialParameters
   parameterName=FortniteActiveTimeOfDay
  }
  n38 = VectorParameter() {desc=code controled, group=Outline, parameterName=GamePlayOutlineColor, defaultValue=<rgba 1 0.425 0.17 1>}
  n39 = Multiply(A=DotProduct(A=@n37, B=Constant4Vector() {constant=<rgba 1 1 1 0.4>}.RGBA), B=@n38.RGB)
  n40 = SceneTexture() {sceneTextureId=PPI_CustomDepth}
  n41 = Divide(A=Subtract(A=ComponentMask(None=@n40.Color) {r=true}, B=Constant() {desc=start distance, r=2250}), B=Constant() {desc=fade distance, r=500})
  n42 = CollectionParameter() {
   collection=&Stencil_ID_Definitions
   parameterName=Green_OutlineID_CodeReferenced
  }
  n43 = Multiply(
   A=OneMinus(None=Clamp(None=@n41))
   B=MaterialFunctionCall(ID=@n42) {materialFunction=&StencilMaskCompare}.Result
  )
  n44 = ComponentMask(None=LinearInterpolate(
    A=LinearInterpolate(A=@n1, B=@n25, Alpha=@n35)
    B=LinearInterpolate(A=@n36, B=@n39, Alpha=@n24) {desc=outline z tested objects when opaque}
    Alpha=Multiply(A=@n31, B=@n43)
   )) {b=true, g=true, r=true}
  n45 = ScalarParameter() {parameterName=desaturation}
  n46 = Constant()
  n47 = If(
   A=ComponentMask(None=SceneTexture() {sceneTextureId=PPI_ShadingModelColor}.Color) {r=true}
   B=Constant() {r=0.2}
   "A > B"=@n46
   "A == B"=Constant() {r=1}
   "A < B"=@n46
  )
  n48 = ComponentMask(None=SceneTexture() {sceneTextureId=PPI_Opacity}.Color) {r=true}
  n49 = Multiply(A=Multiply(A=Desaturation(None=@n44, Fraction=@n45), B=Multiply(A=@n47, B=@n48)), B=ScalarParameter() {parameterName=Skin Emissive})
  n50 = OneMinus(None=ComponentMask(None=SceneTexture() {sceneTextureId=PPI_Metallic}.Color) {r=true})
  Add(A=@n44, B=Clamp(None=Multiply(A=@n49, B=@n50)) {clampMode=CMODE_ClampMin})
  n51 = SceneTexture() {sceneTextureId=PPI_PostProcessInput0}
  n52 = CollectionParameter() {collection=&FortniteMaterialParameters, parameterName=SharpenFilterAmount}
  n53 = Add(A=@n52)
  n54 = SceneTexture() {sceneTextureId=PPI_PostProcessInput0}
  n55 = Constant2Vector() {g=-1}
  n56 = TextureCoordinate()
  n57 = SceneTexture(UVs=Add(A=Multiply(A=@n54.InvSize, B=@n55), B=@n56)) {sceneTextureId=PPI_PostProcessInput0}
  n58 = Constant2Vector() {r=-1}
  n59 = SceneTexture(UVs=Add(A=Multiply(A=@n54.InvSize, B=@n58), B=@n56)) {sceneTextureId=PPI_PostProcessInput0}
  n60 = Constant2Vector() {r=1}
  n61 = SceneTexture(UVs=Add(A=Multiply(A=@n54.InvSize, B=@n60), B=@n56)) {sceneTextureId=PPI_PostProcessInput0}
  n62 = Constant2Vector() {g=1}
  n63 = SceneTexture(UVs=Add(A=Multiply(A=@n54.InvSize, B=@n62), B=@n56)) {sceneTextureId=PPI_PostProcessInput0}
  n64 = Multiply(A=Add(A=Add(A=@n57.Color, B=@n59.Color), B=Add(A=@n61.Color, B=@n63.Color))) {constB=-0.25}
  ComponentMask(None=Add(A=Multiply(A=@n51.Color, B=@n53), B=Multiply(A=@n52, B=@n64))) {b=true, g=true, r=true}
  n65 = CollectionParameter() {
   collection=&FortniteMaterialParameters
   parameterName=CreativeModeDisablePBWOutlines
  }
  LinearInterpolate(A=@n44, B=ComponentMask(None=SceneTexture() {sceneTextureId=PPI_PostProcessInput0}.Color) {b=true, g=true, r=true}, Alpha=@n65)
  MaterialFunctionCall!(a=Desaturation(None=@n51.Color), b=ScalarParameter() {defaultValue=5, parameterName=TestValue}) {materialFunction=/Game/Packages/User_Built_Props/EqualCheck}
  Multiply(A=@n14) {constB=4}
  Multiply(A=@n35, B=@n34.Result)
  ScalarParameter() {defaultValue=3, parameterName=NumberOfPixels}
  ScalarParameter() {defaultValue=1, parameterName=Sharpen}
 ]
}
