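# Post-process material that draws outlines and tints on stencil-tagged objects, such as selections and colored overlays, blended translucently before depth of field.
# Four resolution-scaled neighbor taps of custom depth form the edge, while stencil ID compares choose the mask and color.
# Offset custom depth versus scene depth limits selection outlines, with a faint pulsing fill, to visible areas.
# Blue, red and yellow overlays instead show where objects are occluded, with a dim fill and brighter edges.
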Material /Game/Packages/Fortnite_UniqueMaterials_1/SkinAndMeshOutlinePostProcess_BR_Optimized
graph {
 properties {
  bEnableResponsiveAA=true
  bEnableStencilTest=true
  blendMode=BLEND_Translucent
  blendableLocation=BL_SceneColorBeforeDOF
  blendableOutputAlpha=true
  blendablePriority=-5
  materialDomain=MD_PostProcess
  pixelDepthOffsetMode=PDOM_Legacy
  shadingModel=MSM_Unlit
  stencilCompare=MSC_NotEqual
  usedShadingModels=Unlit
 }
 nodes [
  n1 = Add(A=SceneTexture() {sceneTextureId=PPI_CustomDepth}.Color, B=ScalarParameter() {defaultValue=-2, parameterName=Custom Depth Offset Amount})
  n2 = SceneTexture() {sceneTextureId=PPI_SceneDepth}
  n3 = Clamp(None=ComponentMask(None=Subtract(A=@n1, B=@n2.Color)) {r=true})
  n4 = CollectionParameter() {
   collection=/Game/Packages/Fortress_SharedMaterials/Stencil_ID_Definitions/Stencil_ID_Definitions
   parameterName=SelectionOutlineID_CodeReferenced
  }
  n5 = CollectionParameter() {
   collection=&Stencil_ID_Definitions
   parameterName=SelectionOutlineID2_CodeReferenced
  }
  n6 = MaterialFunctionCall!(ID=@n5) {materialFunction=/Game/Packages/Fortress_SharedMaterials/Stencil_ID_Definitions/StencilMaskCompare}
  n7 = Multiply(
   A=OneMinus(None=Ceil(None=@n3) {desc=z tested objects black z tested - everything else white}) {desc=occlusion mask}
   B=Max(
    A=MaterialFunctionCall(ID=@n4) {materialFunction=&StencilMaskCompare}.Result
    B=@n6.Result
   )
  )
  n8 = Sine(None=Multiply(A=Time(), B=ScalarParameter() {defaultValue=0.75, parameterName=PulseRate}))
  n9 = ComponentMask(None=ScreenPosition().ViewportUV) {g=true, r=true}
  n10 = Subtract(A=Divide(A=ViewSize(), B=Constant() {r=320}), B=Constant() {r=2}) {constB=0}
  n11 = Multiply(A=Clamp(None=Ceil(None=@n10)) {maxDefault=2, minDefault=1}, B=SceneTexelSize()) {constA=2}
  n12 = Constant()
  n13 = AppendVector(A=ComponentMask(None=@n11) {desc=x, r=true}, B=@n12)
  n14 = Add(A=@n9, B=@n13) {desc=x+}
  n15 = AppendVector(A=@n12, B=ComponentMask(None=@n11) {desc=y, g=true}) {desc=y+}
  n16 = Add(A=@n9, B=@n15) {desc=y+}
  n17 = Constant() {r=-1}
  n18 = Add(A=@n9, B=Multiply(A=@n13, B=@n17)) {desc=x-}
  n19 = Add(A=@n9, B=Multiply(A=@n17, B=@n15)) {desc=y-}
  n20 = ComponentMask(None=MaterialFunctionCall!(
    Base=SceneTexture() {sceneTextureId=PPI_CustomDepth}.Color
    BaseSampleBias
    Scale
    Sample1=SceneTexture(UVs=@n14) {sceneTextureId=PPI_CustomDepth}.Color
    Sample2=SceneTexture(UVs=@n16) {sceneTextureId=PPI_CustomDepth}.Color
    Sample3=SceneTexture(UVs=@n18) {sceneTextureId=PPI_CustomDepth}.Color
    Sample4=SceneTexture(UVs=@n19) {sceneTextureId=PPI_CustomDepth}.Color
    Divisor=Constant() {r=-16}
    UseCenteredLines
    UseRGB
   ) {materialFunction=/Game/Packages/jl_tests/combine4Samples}.Result) {r=true}
  n21 = Add(A=LinearInterpolate(Alpha=ConstantBiasScale(None=@n8)) {constA=0.005, constB=0.02}, B=@n20)
  n22 = Constant() {r=8}
  n23 = Subtract(A=@n2.Color, B=@n1)
  n24 = OneMinus(None=Ceil(None=Clamp(None=@n23)) {desc=z tested objects black z tested - everything else white}) {desc=occlusion mask}
  n25 = ComponentMask(None=Multiply(
    A=MaterialFunctionCall(ID=@n22) {materialFunction=&StencilMaskCompare}.Result
    B=@n24
   )) {r=true}
  n26 = CollectionParameter() {collection=&Stencil_ID_Definitions, parameterName=RedOverlayID}
  n27 = ComponentMask(None=Multiply(
    A=MaterialFunctionCall(ID=@n26) {materialFunction=&StencilMaskCompare}.Result
    B=@n24
   )) {r=true}
  n28 = CollectionParameter() {collection=&Stencil_ID_Definitions, parameterName=YellowOverlayID}
  n29 = ComponentMask(None=Multiply(
    A=MaterialFunctionCall(ID=@n28) {materialFunction=&StencilMaskCompare}.Result
    B=@n24
   )) {r=true}
  n30 = Clamp(None=Add(A=Add(A=@n25, B=@n27), B=@n29))
  n31 = LinearInterpolate(Alpha=@n20) {constA=0.2}
  Clamp(None=Add(A=Multiply(A=@n7, B=@n21), B=Multiply(A=@n30, B=@n31)))
  n32 = VectorParameter() {desc=code controled, group=Outline, parameterName=OutlineColor, defaultValue=<rgba 0.7552 0.6719 0.6178 0>}
  n33 = VectorParameter() {desc=code controled, group=Outline, parameterName=OutlineColor2, defaultValue=<rgba 0.059 0.7552 0.5574 0>}
  n34 = Constant() {r=19}
  n35 = LinearInterpolate(
   A=LinearInterpolate(
    A=@n32.RGB
    B=@n33.RGB
    Alpha=MaterialFunctionCall(ID=@n34) {materialFunction=&StencilMaskCompare}.Result
   )
   B=Constant3Vector() {constant=<rgba 0.113 0.3263 0.5208 0>}
   Alpha=@n25
  )
  n36 = Constant3Vector() {constant=<rgba 0.8 0 0 0>}
  MaterialFunctionCall(Emissive=LinearInterpolate(A=LinearInterpolate(A=@n35, B=@n36, Alpha=@n27), B=Constant3Vector() {constant=<rgba 1 0.8181 0 0>}, Alpha=@n29)) {materialFunction=&MF_FX_EmissiveAdjustments}
  Multiply(A=@n7, B=@n6.Result)
 ]
}
