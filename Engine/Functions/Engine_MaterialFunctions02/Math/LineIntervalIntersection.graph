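# LineIntervalIntersection is a math material function, exposed to the library, that finds where two lines sharing endpoint X values intersect.
# It divides the endpoint gap at A by the summed A and B gaps for x, then blends Line1's Y values for y, outputting x, y and xy.
# The A.x/B.x remap Lerp feeds no output, so x stays a normalized fraction.
# An unwired Custom HLSL node holds the same math for reference.
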MaterialFunction /Engine/Functions/Engine_MaterialFunctions02/Math/LineIntervalIntersection
graph {
 properties {
  bExposeToLibrary=true
  description=Returns the point of intersection for two lines. The X value for the endpoints of both lines must match. A Y value will be calculated from the interval.
  libraryCategoriesText [Misc,Math]
 }
 nodes [
  Custom!(line1A, line1B, line2A, line2B) {
   code="float a =line2B - line1B;\r\nfloat b =a+(line1A - line2A);\r\nfloat2 intersect=0;\r\n\r\n//the shared x value\r\nfloat invx=a/b;\r\nfloat yintersect =(line1A*invx)+(line1B*(1-invx));\r\nintersect.x=1-invx;\r\nintersect.y=yintersect;\r\n\r\nreturn intersect;"
   desc=For reference
   outputType=CMOT_Float2
   inputs [
    {inputName=line1A, input {expression=null, mask=1, maskR=1, outputIndex=1}}
    {inputName=line1B, input {expression=null, mask=1, maskG=1, outputIndex=2}}
    {inputName=line2A, input {expression=null, mask=1, maskR=1, outputIndex=1}}
    {inputName=line2B, input {expression=null, mask=1, maskG=1, outputIndex=2}}
   ]
  }
  n1 = FunctionInput() {
   description=The Y value of point A on the second line
   inputName=Line2 - A.y
   inputType=FunctionInput_Scalar
   sortPriority=2
   previewValue=<1 0 0 1>
  }
  n2 = FunctionInput() {description=The Y value of point A on the first line, inputName=Line1 - A.y, inputType=FunctionInput_Scalar, previewValue=<0.9 0 0 1>}
  n3 = Subtract(A=@n1, B=@n2) {desc=b}
  n4 = FunctionInput() {description=The Y value of point B on the first line, inputName=Line1 - B.y, inputType=FunctionInput_Scalar, sortPriority=1}
  n5 = FunctionInput() {
   description=The Y value of point B on the second line
   inputName=Line2 - B.y
   inputType=FunctionInput_Scalar
   sortPriority=3
   previewValue=<0.1 0 0 1>
  }
  n6 = Divide(A=@n3, B=Add(A=@n3, B=Subtract(A=@n4, B=@n5) {desc=a}))
  n7 = Add(A=Multiply(A=@n4, B=@n6), B=Multiply(A=@n2, B=OneMinus(None=@n6) {desc=X})) {desc=Y}
  FunctionOutput(None=AppendVector(A=FunctionOutput(None=@n6) {outputName=x, sortPriority=1}, B=FunctionOutput(None=@n7) {outputName=y, sortPriority=1})) {outputName=xy}
  n8 = FunctionInput() {
   bUsePreviewValueAsDefault=true
   description=The x value of both A points to sample from. In most cases leave at 0.
   inputName=A.x
   inputType=FunctionInput_Scalar
   sortPriority=4
  }
  n9 = FunctionInput() {
   bUsePreviewValueAsDefault=true
   description="The x value of both B points to sample from.  In most cases leave at 1."
   inputName=B.x
   inputType=FunctionInput_Scalar
   sortPriority=5
   previewValue=<1 0 0 1>
  }
  LinearInterpolate(A=@n8, B=@n9, Alpha=@n6)
 ]
}
