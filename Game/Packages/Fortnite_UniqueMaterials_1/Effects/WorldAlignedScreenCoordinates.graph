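# Material function that outputs screen coordinates anchored to a world-space center point, defaulting to the actor position and exposed to the library.
# It subtracts the point's projected buffer UV from the pixel's viewport UV, then multiplies by camera distance to compensate for perspective scaling.
# A final X scale reads buffer resolution but divides R by itself, so it evaluates to 1 and applies no aspect correction.
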MaterialFunction /Game/Packages/Fortnite_UniqueMaterials_1/Effects/WorldAlignedScreenCoordinates
graph {
 properties {bExposeToLibrary=true}
 nodes [
  n1 = FunctionInput(Preview=ActorPositionWS()) {bUsePreviewValueAsDefault=true, inputName=Center Point World Position}
  n2 = Distance(A=CameraPositionWS(), B=@n1)
  n3 = Subtract(
   A=ScreenPosition().ViewportUV
   B=MaterialFunctionCall(World Position=@n1) {materialFunction=&TransformToClipSpace}.Buffer UVs for Scene Textures
  )
  n4 = MaterialFunctionCall(Float2=MaterialFunctionCall() {materialFunction=&ScreenResolution}.Buffer Resolution) {materialFunction=&BreakOutFloat2Components}
  n5 = Divide(A=@n4.R, B=@n4.R)
  n6 = Constant() {r=1}
  FunctionOutput(None=Multiply(A=Multiply(A=@n2, B=@n3), B=AppendVector(A=@n5, B=@n6)))
 ]
}
