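# Material function that produces a sparkle mask from the Slowfield sparkle noise texture.
# Screen-space UVs are anchored to a world center point (component pivot by default), compensated for view width and FOV, then multiplied by a Scale input.
# A static R sample at a slightly different tiling multiplies a panning G sample, and the product is boosted.
# Outputs the mask and the pre-Scale UVs; a LineIntervalIntersection call is left unused.
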MaterialFunction /Game/Packages/Fortnite_UniqueMaterials_1/Effects/FortniteSparkle
graph {
 properties {bExposeToLibrary=true}
 nodes [
  n1 = FunctionInput() {bUsePreviewValueAsDefault=true, inputName=Scale, inputType=FunctionInput_Scalar, previewValue=<-0.1551 0 0 1>}
  n2 = MaterialFunctionCall() {materialFunction=&ComponentPivotLocation}
  n3 = Divide(A=ComponentMask(None=ViewSize()) {r=true}) {constB=2048}
  n4 = Multiply(
   A=MaterialFunctionCall!(Center Point World Position=FunctionInput(Preview=@n2.Result) {bUsePreviewValueAsDefault=true, inputName=Sparkle World center point}) {materialFunction=&/WorldAlignedScreenCoordinates}.Result
   B=Multiply(A=Divide(B=@n3), B=MaterialFunctionCall() {materialFunction=/Engine/Functions/Engine_MaterialFunctions02/Utility/FOV}."FOV ")
  )
  n5 = Multiply(A=@n1, B=@n4)
  n6 = TextureSample(UVs=Multiply(A=@n5) {constB=1.1}) {samplerType=SAMPLERTYPE_LinearColor, texture=/Game/Gadgets/Assets/Slow_Field/Textures/T_Slowfield_SparkleNoise_01}
  n7 = TextureSample(UVs=Panner(Coordinate=@n5) {speedX=0.1, speedY=0.1}) {samplerType=SAMPLERTYPE_LinearColor, texture=&T_Slowfield_SparkleNoise_01}
  FunctionOutput(None=Multiply(A=Multiply(A=@n6.R, B=@n7.G)) {constB=10})
  FunctionOutput(None=@n4) {outputName=UVS, sortPriority=1}
  MaterialFunctionCall!(Line1 - A.y, Line1 - B.y, Line2 - A.y, Line2 - B.y, A.x, B.x) {materialFunction=/Engine/Functions/Engine_MaterialFunctions02/Math/LineIntervalIntersection}
 ]
}
